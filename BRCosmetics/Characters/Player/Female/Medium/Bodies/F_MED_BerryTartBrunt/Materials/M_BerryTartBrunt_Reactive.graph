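# Music-reactive skinned character material for the BerryTartBrunt outfit, animating emissive lines, rings and a belt glow.
# MF_AudioReactivity amplitude and spectrum feed line brightness and offsets of Frac-and-SphereMask bands panning along gradient textures.
# Fresnel sine tinting, world-aligned swirl noise, a looping belt dot and time-of-day scaling shape the glow.
# Static switches gate reactivity, debug views and a diamond layer between the CharacterParent 2023 PreFX and PostFX functions.
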Material /BRCosmetics/Characters/Player/Female/Medium/Bodies/F_MED_BerryTartBrunt/Materials/M_BerryTartBrunt_Reactive
graph {
 properties {
  bUseMaterialAttributes=true
  bUsedWithMeshDeformer=true
  bUsedWithMorphTargets=true
  bUsedWithSkeletalMesh=true
  pixelDepthOffsetMode=PDOM_Legacy
  shadingModel=MSM_FromMaterialExpression
  usedShadingModels=Default Lit | Subsurface | Clear Coat | Subsurface Profile | Cloth
 }
 nodes [
  n1 = Multiply(A=Time(), B=ScalarParameter() {defaultValue=-1.2, group=Reactive_Lines, parameterName=PanningSpeed})
  Add(A=@n1, B=Multiply(A=ScalarParameter() {defaultValue=-0.5, group=Reactive_Lines, parameterName=PanningSpeed_Offset}, B=NamedRerouteUsage()))
  n2 = MaterialFunctionCall() {materialFunction=&MF_CharacterParent_2023_PreFX}
  n3 = GetMaterialAttributes(None=@n2.Result) {attributeGetTypes [#1, #2, #3]}
  n4 = NamedRerouteDeclaration(None=@n3.Base Color) {name=BaseColorIN, nodeColor=<rgba 1 0.3516 0 1>}
  n5 = Saturate(None=Multiply(A=@n4) {constB=2})
  n6 = ChannelMaskParameter(None=NamedRerouteDeclaration(None=@n2.Tech Art Mask) {name=TechArtMask, nodeColor=<rgba 1 0.1641 0 1>}) {group=Reactive, parameterName=ReactiveLines}
  n7 = NamedRerouteUsage()
  n8 = NamedRerouteUsage()
  n9 = StaticSwitchParameter(
   True=SetMaterialAttributes!(
    MaterialAttributes=@n3.MaterialAttributes
    Base Color=LinearInterpolate(A=@n4, B=@n5, Alpha=@n6.R)
    Emissive Color=StaticSwitchParameter(True=@n7, False=@n8) {defaultValue=true, group=Reactive, parameterName=UseReactivity}
   ) {attributeSetTypes [#1, #2]}
   False=@n2.Result
  ) {defaultValue=true, group=TechArt_Add, parameterName=UseMusicReactivity}
  n10 = NamedRerouteUsage()
  MaterialFunctionCall(
   Material=StaticSwitchParameter(
    True=MaterialFunctionCall!(Material=@n9, TechArtMask=@n10) {materialFunction=&/MF_NewDiamond_Cheap}.Material
    False=@n9
   ) {defaultValue=true, group=TechArt_Add, parameterName=UseDiamond}
   M=@n2.M
   Pre Skinned Local Position=@n2.M
   Distance Blend=@n2.M
   Fresnel=@n2.M
  ) {materialFunction=&MF_CharacterParent_2023_PostFX}
  NamedRerouteDeclaration(None=@n3.Emissive Color) {name=EM_IN, nodeColor=<rgba 0.2891 1 0 1>}
  n11 = NamedRerouteUsage()
  n12 = ScalarParameter() {defaultValue=4, group=Reactive, parameterName=Part_POW02}
  n13 = Multiply(
   A=SmoothStep(Value=Power(Base=@n11, Exp=@n12)) {constMin=0.5}
   B=ScalarParameter() {defaultValue=0.15, group=Reactive, parameterName=Gradient_Offset_Music02}
  )
  n14 = Time()
  n15 = ScalarParameter() {defaultValue=-0.35, group=Reactive_Lines, parameterName=PanningSpeed_Moving_leg}
  n16 = Add(A=@n13, B=Multiply(A=@n14, B=@n15))
  NamedRerouteDeclaration(None=@n16) {name=Time01, nodeColor=<rgba 0.875 1 0 1>}
  NamedRerouteDeclaration(None=Add(A=@n16, B=NamedRerouteDeclaration(None=@n15) {name=TimeGap, nodeColor=<rgba 0.4062 1 0 1>})) {name=Time02, nodeColor=<rgba 0.8047 0 1 1>}
  NamedRerouteDeclaration(None=@n2.Pre skinned Local Position) {name=PreskinnedPOS, nodeColor=<rgba 0 1 0.9063 1>}
  n17 = NamedRerouteUsage()
  n18 = NamedRerouteUsage()
  n19 = NamedRerouteUsage()
  n20 = NamedRerouteUsage()
  n21 = NamedRerouteUsage()
  n22 = LinearInterpolate(Alpha=NamedRerouteUsage()) {constA=1.5}
  n23 = ScalarParameter() {defaultValue=3, group=Reactive, parameterName=NoiseMax_Music}
  n24 = Power(Base=NamedRerouteUsage(), Exp=ScalarParameter() {defaultValue=12, group=Reactive, parameterName=Line_Part_POW01}) {constExponent=10}
  n25 = ScalarParameter() {defaultValue=1, group=Reactive, parameterName=NoiseMin}
  n26 = ScalarParameter() {defaultValue=5, group=Reactive, parameterName=NoiseMax}
  n27 = NamedRerouteUsage()
  n28 = VectorParameter() {group=Reactive_Lines, parameterName=LineTint_LOW, channelNames {a=Intensity, b="", g="", r=""}, defaultValue=<rgba 1 0 0.6632 1>}
  n29 = Multiply(A=@n28, B=@n28)
  n30 = VectorParameter() {group=Reactive_Lines, parameterName=LineTint_HIGH, channelNames {a=Intensity, b="", g="", r=""}, defaultValue=<rgba 1 0 0.6632 10>}
  n31 = Multiply(A=@n30, B=@n30)
  n32 = Transform(None=MaterialFunctionCall(Normal=NamedRerouteUsage(), Flatness=Constant() {r=-1}) {materialFunction=&FlattenNormal}.Result)
  n33 = Sine(None=Add(A=Fresnel(Normal=@n32) {baseReflectFraction=0, exponent=1}, B=@n1))
  n34 = Multiply(
   A=Add(A=Multiply(A=@n23, B=@n24), B=LinearInterpolate(A=@n25, B=@n26, Alpha=@n27) {constA=1, constB=2})
   B=LinearInterpolate(
    A=NamedRerouteDeclaration(None=@n29) {name=LowColor, nodeColor=<rgba 0.1719 0 1 1>}
    B=NamedRerouteDeclaration(None=@n31) {name=HighColor, nodeColor=<rgba 0.1484 1 0 1>}
    Alpha=ConstantBiasScale(None=@n33)
   ) {constA=1}
  )
  n35 = Add(
   A=Multiply(A=Power(Base=@n21, Exp=@n22) {constExponent=1.5}, B=@n34)
   B=Multiply(A=NamedRerouteDeclaration(None=@n34) {name=Color_EM, nodeColor=<rgba 0.5703 1 0 1>}, B=NamedRerouteUsage())
  )
  n36 = Multiply(
   A=StaticSwitchParameter(True=Constant() {r=1}, False=SmoothStep(Value=@n19) {constMax=0.3}) {group=Reactive, parameterName=DebugReactivity}
   B=Add(A=NamedRerouteUsage(), B=Add(A=@n20, B=@n35))
  )
  n37 = Constant() {r=1}
  n38 = Constant() {r=0.7}
  n39 = MaterialFunctionCall(Value_Morning=@n37, Value_Day=@n37, Value_Evening=@n38, Value_Night=@n38) {materialFunction=&Scalar_Time_of_Day}
  NamedRerouteDeclaration(None=StaticSwitchParameter(True=StaticSwitchParameter(True=@n17, False=@n18) {group=Reactive, parameterName=DebugAmplitude02}, False=Multiply(A=@n36, B=@n39.Result)) {group=Reactive, parameterName=DebugAmplitude}) {name=EM_OUT, nodeColor=<rgba 1 0.8906 0 1>}
  n40 = MaterialFunctionCall() {materialFunction=/Game/Packages/Fortress_SharedMaterials/Base_Material_Functions/BaseCharacterShaderFunctions/Audio/MF_AudioReactivity}
  NamedRerouteDeclaration(None=LinearInterpolate(A=@n40.Amplitude, B=Clamp(None=@n40.Amplitude) {minDefault=0.35}, Alpha=ScalarParameter() {parameterName=EnableFrontendFakeAudioLoop})) {name=Amplitude, nodeColor=<rgba 0 0.25 1 1>}
  n41 = Multiply(A=Time(), B=ScalarParameter() {defaultValue=-6, group=Reactive_Lines, parameterName=SwirlSpeed})
  n42 = Constant3Vector() {constant=<rgba 0 0 1 0>}
  n43 = MaterialFunctionCall(
   TextureObject=TextureObjectParameter() {group=Reactive_Lines, parameterName=SwirlVector, texture=/Game/Global/Textures/Default/Noise/T_Noise_Cloud}
   TextureSize=ScalarParameter() {defaultValue=20, group=Reactive_Lines, parameterName=SwirlSize_L}
   WorldPosition=Add(A=NamedRerouteUsage(), B=Multiply(A=@n41, B=@n42))
   World Space Normal=VertexInterpolator(VS=PreSkinnedNormal()).PS
  ) {materialFunction=&WorldAlignedTexture}
  NamedRerouteDeclaration(None=Power(
    Base=ComponentMask(None=@n43.XYZ Texture) {r=true}
    Exp=ScalarParameter() {defaultValue=1.5, group=Reactive_Gradient, parameterName=NoiseBlend_POW}
   )) {name=Noise_Blend, nodeColor=<rgba 1 0 0.3047 1>}
  n44 = NamedRerouteUsage()
  n45 = SphereMask(A=@n44, B=Constant() {r=1}) {attenuationRadius=0.02, hardnessPercent=0}
  n46 = Power(Base=Multiply(A=OneMinus(None=@n45), B=@n44)) {constExponent=2.2}
  NamedRerouteDeclaration(None=@n46) {name=Fade, nodeColor=<rgba 1 0.0234 0 1>}
  NamedRerouteDeclaration(None=@n24) {name=DebugAmplitude, nodeColor=<rgba 1 0.7266 0 1>}
  n47 = ScalarParameter() {defaultValue=-0.5, group=Reactive_Lines, parameterName=PanningSpeed_Moving_leg02}
  n48 = Add(A=@n13, B=Multiply(A=@n14, B=@n47))
  NamedRerouteDeclaration(None=@n48) {name=Time03, nodeColor=<rgba 0.875 1 0 1>}
  NamedRerouteDeclaration(None=@n6.R) {name=ReactiveLinesArea, nodeColor=<rgba 1 0 0.6094 1>}
  NamedRerouteDeclaration(None=Add(A=@n48, B=@n47)) {name=Time04, nodeColor=<rgba 0.8047 0 1 1>}
  n49 = TextureSampleParameter2D() {
   group=Reactive_Gradient
   parameterName=Belt_Line
   samplerType=SAMPLERTYPE_Masks
   texture=/BRCosmetics/Characters/Player/Female/Medium/Bodies/F_MED_BerryTartBrunt/Textures/T_Belt_Gradient
  }
  n50 = NamedRerouteDeclaration(None=@n49.R) {name=BeltMask, nodeColor=<rgba 1 0.4922 0 1>}
  n51 = Multiply(
   A=Multiply(A=NamedRerouteUsage(), B=@n50)
   B=LinearInterpolate(
    A=ScalarParameter() {defaultValue=4, group=Reactive_Gradient, parameterName=BasicBeltEM_Amplitude}
    B=ScalarParameter() {defaultValue=0.5, group=Reactive_Gradient, parameterName=BasicBeltEM_LOW_Amplitude}
    Alpha=@n49.B
   )
  ) {constB=5}
  n52 = Multiply(
   A=@n50
   B=LinearInterpolate(
    A=ScalarParameter() {defaultValue=1, group=Reactive_Gradient, parameterName=BasicBeltEM}
    B=ScalarParameter() {defaultValue=1, group=Reactive_Gradient, parameterName=BasicBeltEM_LOW}
    Alpha=@n49.B
   )
  ) {constB=0.6}
  n53 = OneMinus(None=@n49.B)
  n54 = Multiply(A=Time(), B=ScalarParameter() {defaultValue=0.7, group=Reactive_Lines, parameterName=PanningSpeed_Belt})
  n55 = Subtract(
   A=Multiply(A=Frac(None=@n54), B=ScalarParameter() {defaultValue=1.4, group=Reactive_Gradient, parameterName=PhaseMulti}) {constB=1.6}
   B=ScalarParameter() {group=Reactive_Gradient, parameterName=PhaseOffset}
  ) {constB=-0.3}
  n56 = ScalarParameter() {defaultValue=0.2, group=Reactive_Gradient, parameterName=DotRadius}
  n57 = SphereMask(A=@n49.G, B=@n55, Radius=@n56) {attenuationRadius=0.3, hardnessPercent=0}
  n58 = SphereMask(A=@n49.G, B=@n55, Radius=@n56) {attenuationRadius=0.3, hardnessPercent=0}
  n59 = Multiply(A=Power(Base=Multiply(A=@n50, B=@n53)), B=Add(A=Power(Base=@n57), B=Power(Base=@n58)))
  n60 = ScalarParameter() {defaultValue=20, group=Reactive_Gradient, parameterName=Dot_Intensity}
  NamedRerouteDeclaration(None=Add(A=Add(A=@n51, B=@n52), B=Multiply(A=@n59, B=@n60) {constB=10})) {name=Belt_Animate_Mask, nodeColor=<rgba 1 0 0.5625 1>}
  NamedRerouteDeclaration(None=SmoothStep(Value=ComponentMask(None=@n43.XYZ Texture) {b=true}) {constMax=0.3}) {name=NoiseBlend_S, nodeColor=<rgba 1 0.8438 0 1>}
  NamedRerouteDeclaration(None=@n40.Spectrum) {name=Spectrum, nodeColor=<rgba 1 0 0.0703 1>}
  n61 = SmoothStep(Value=Power(Base=NamedRerouteUsage(), Exp=ScalarParameter() {defaultValue=4, group=Reactive, parameterName=Part_POW01})) {constMin=0.5}
  NamedRerouteDeclaration(None=@n61) {name=DebugAmplitude02, nodeColor=<rgba 0 1 0.5547 1>}
  NamedRerouteDeclaration(None=MaterialFunctionCall!(
    Blend Range=ScalarParameter() {defaultValue=300, group=Reactive_Override, parameterName=DistanceRange}
    Start Offset=ScalarParameter() {defaultValue=-100, group=Reactive_Override, parameterName=DistanceStartOffset}
   ) {materialFunction=/Engine/Functions/Engine_MaterialFunctions03/Particles/Distance_Blend}.Result) {name=DistanceBlend, nodeColor=<rgba 0 1 0.1328 1>}
  NamedRerouteDeclaration(None=@n3.Normal) {name=Normal_IN, nodeColor=<rgba 0 0.9297 1 1>}
  n62 = TextureSampleParameter2D() {
   group=Reactive_Gradient
   parameterName=Gradient
   samplerType=SAMPLERTYPE_LinearColor
   texture=/BRCosmetics/Characters/Player/Female/Medium/Bodies/F_MED_BerryTartBrunt/Textures/T_Gradient_FX
  }
  NamedRerouteDeclaration(None=@n62.R) {name=Gradient, nodeColor=<rgba 1 0.6563 0 1>}
  NamedRerouteDeclaration(None=@n62.G) {name=Mask, nodeColor=<rgba 0.3359 0 1 1>}
  n63 = VectorParameter() {group=Reactive_Lines, parameterName=MovingLineTint_HIGH_01, channelNames {a=Intensity, b="", g="", r=""}, defaultValue=<rgba 1 0 0.6632 10>}
  n64 = NamedRerouteUsage()
  n65 = NamedRerouteUsage()
  n66 = Constant() {r=1}
  n67 = Multiply(
   A=Power(
    Base=SphereMask(A=@n65, B=@n66) {attenuationRadius=0.4, hardnessPercent=0}
    Exp=ScalarParameter() {defaultValue=2, group=Reactive_Lines, parameterName=RingPow}
   )
   B=ScalarParameter() {defaultValue=0.2, group=Reactive_Lines, parameterName=RingIntensity}
  ) {constB=0.2}
  n68 = NamedRerouteUsage()
  n69 = Constant() {r=1}
  n70 = Multiply(A=OneMinus(None=SphereMask(A=@n68, B=@n69) {attenuationRadius=0.02, hardnessPercent=0}), B=@n68)
  n71 = ScalarParameter() {defaultValue=-0.7, group=Reactive_Lines, parameterName=PanningSpeed_Moving}
  n72 = Add(
   A=Multiply(A=@n61, B=ScalarParameter() {defaultValue=0.25, group=Reactive, parameterName=Gradient_Offset_Music01})
   B=Multiply(A=Time(), B=@n71)
  )
  n73 = Frac(None=@n72)
  n74 = ScalarParameter() {defaultValue=0.1, parameterName=MovingLine_Thickness}
  n75 = Multiply(
   A=Multiply(A=Power(Base=@n70) {constExponent=2.2}, B=SphereMask(A=@n68, B=@n73, Radius=@n74) {attenuationRadius=0.1, hardnessPercent=0})
   B=NamedRerouteUsage()
  )
  n76 = NamedRerouteUsage()
  n77 = Constant() {r=1}
  n78 = Multiply(A=OneMinus(None=SphereMask(A=@n76, B=@n77) {attenuationRadius=0.02, hardnessPercent=0}), B=@n76)
  n79 = Frac(None=Add(A=@n72, B=@n71))
  n80 = Multiply(
   A=Multiply(A=Power(Base=@n78) {constExponent=2.2}, B=SphereMask(A=@n76, B=@n79, Radius=@n74) {attenuationRadius=0.1, hardnessPercent=0})
   B=NamedRerouteUsage()
  )
  n81 = Multiply(A=Multiply(A=@n63, B=@n63), B=Add(A=Multiply(A=@n64, B=@n67), B=Add(A=@n75, B=@n80)))
  n82 = VectorParameter() {group=Reactive_Lines, parameterName=MovingLineTint_HIGH_02, channelNames {a=Intensity, b="", g="", r=""}, defaultValue=<rgba 1 0 0.6632 10>}
  n83 = NamedRerouteUsage()
  n84 = NamedRerouteUsage()
  n85 = Constant() {r=1}
  n86 = Multiply(
   A=Power(
    Base=SphereMask(A=@n84, B=@n85) {attenuationRadius=0.2, hardnessPercent=0}
    Exp=ScalarParameter() {defaultValue=2, group=Reactive_Lines, parameterName=RingPow_Leg}
   )
   B=ScalarParameter() {defaultValue=0.2, group=Reactive_Lines, parameterName=RingIntensity_Leg}
  ) {constB=0.2}
  n87 = Frac(None=NamedRerouteUsage())
  n88 = ScalarParameter() {defaultValue=0.03, parameterName=MovingLine_Thickness_leg}
  n89 = Multiply(A=Multiply(A=@n46, B=SphereMask(A=@n44, B=@n87, Radius=@n88) {attenuationRadius=0.1, hardnessPercent=0}), B=NamedRerouteUsage())
  n90 = NamedRerouteUsage()
  n91 = Multiply(A=NamedRerouteUsage(), B=SphereMask(A=NamedRerouteUsage(), B=Frac(None=@n90), Radius=@n88) {attenuationRadius=0.1, hardnessPercent=0})
  n92 = NamedRerouteUsage()
  n93 = NamedRerouteUsage()
  n94 = Multiply(A=NamedRerouteUsage(), B=SphereMask(A=NamedRerouteUsage(), B=Frac(None=@n93), Radius=@n88) {attenuationRadius=0.1, hardnessPercent=0})
  n95 = NamedRerouteUsage()
  n96 = NamedRerouteUsage()
  n97 = SphereMask(A=NamedRerouteUsage(), B=Frac(None=NamedRerouteUsage()), Radius=@n88) {attenuationRadius=0.1, hardnessPercent=0}
  n98 = Add(A=Add(A=Multiply(A=@n91, B=@n92), B=Multiply(A=@n94, B=@n95)), B=Multiply(A=Multiply(A=@n96, B=@n97), B=NamedRerouteUsage()))
  n99 = Multiply(A=Multiply(A=@n82, B=@n82), B=Add(A=Multiply(A=@n83, B=@n86), B=Add(A=@n89, B=@n98)))
  n100 = NamedRerouteUsage()
  NamedRerouteDeclaration(None=Multiply(A=Add(A=@n81, B=@n99), B=LinearInterpolate(Alpha=@n100) {constA=1, constB=3})) {name=Moving_Pattern, nodeColor=<rgba 1 0 0.4687 1>}
  NamedRerouteDeclaration(None=@n62.B) {name=Gradient_Leg, nodeColor=<rgba 0 0.5547 1 1>}
  NamedRerouteUsage()
  NamedRerouteUsage()
  Power(Base=@n40.Amplitude) {constExponent=3}
 ]
}
