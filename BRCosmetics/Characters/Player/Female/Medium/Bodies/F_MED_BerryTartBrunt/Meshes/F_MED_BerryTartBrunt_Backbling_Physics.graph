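# Physics asset for the F_MED_BerryTartBrunt female character's backbling, setting up bodies for secondary motion of jacket panels and hip laces.
# Seven dyn_jacket capsules each attach to a tiny base capsule under spine_01, with one non-zero swing limit giving hinge-like motion.
# Two-link lace chains on each side hang from base_heart_jnt with swing limits and zero twist.
# Pelvis, spine_01 and thigh capsules form body volumes, joined by free angular constraints.
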PhysicsAsset /BRCosmetics/Characters/Player/Female/Medium/Bodies/F_MED_BerryTartBrunt/Meshes/F_MED_BerryTartBrunt_Backbling_Physics
bodies {
 base_heart_jnt [{shapeType=Capsule, length=0.505, radius=0.505, shapeName=base_heart_jnt_capsule, rotation=<rot 0 -90 0>}]
 base_jacket_bk [{shapeType=Capsule, length=0.505, radius=0.505, shapeName=base_jacket_bk_capsule, rotation=<rot 0 -90 0>}]
 base_jacket_bk_l [{shapeType=Capsule, length=0.505, radius=0.505, shapeName=base_jacket_bk_l_capsule, rotation=<rot 0 -90 0>}]
 base_jacket_bk_r [{shapeType=Capsule, length=0.505, radius=0.505, shapeName=base_jacket_bk_r_capsule, rotation=<rot 0 -90 0>}]
 base_jacket_fr_l [{shapeType=Capsule, length=0.505, radius=0.505, shapeName=base_jacket_fr_l_capsule, rotation=<rot 0 -90 0>}]
 base_jacket_fr_r [{shapeType=Capsule, length=0.505, radius=0.505, shapeName=base_jacket_fr_r_capsule, rotation=<rot 0 -90 0>}]
 base_jacket_l [{shapeType=Capsule, length=0.505, radius=0.505, shapeName=base_jacket_l_capsule, rotation=<rot 0 -90 0>}]
 base_jacket_r [{shapeType=Capsule, length=0.505, radius=0.505, shapeName=base_jacket_r_capsule, rotation=<rot 0 -90 0>}]
 dyn_hip_laces_0_l [{shapeType=Capsule, length=5.6139, radius=1, shapeName=dyn_hip_laces_0_l_capsule, center=<2.5592 0 0>, rotation=<rot 90 0 0>}]
 dyn_hip_laces_0_r [{shapeType=Capsule, length=7.1642, radius=1, shapeName=dyn_hip_laces_0_r_capsule, center=<4.4145 0 0>, rotation=<rot 90 0 0>}]
 dyn_hip_laces_1_l [{shapeType=Capsule, length=2.9202, radius=1.0798, shapeName=dyn_hip_laces_0_l_capsule, center=<2 0 0>, rotation=<rot 90 0 0>}]
 dyn_hip_laces_1_r [{shapeType=Capsule, length=3.9202, radius=1.0798, shapeName=dyn_hip_laces_0_r_capsule, center=<2 0 0>, rotation=<rot 90 0 0>}]
 dyn_jacket_bk [{shapeType=Capsule, length=7, radius=4.5, shapeName=dyn_jacket_bk_capsule, center=<5 0 0>, rotation=<rot 90 0 0>}]
 dyn_jacket_bk_l [{shapeType=Capsule, length=7, radius=4.5, shapeName=dyn_jacket_bk_l_capsule, center=<5 0 0>, rotation=<rot 90 0 0>}]
 dyn_jacket_bk_r [{shapeType=Capsule, length=7, radius=4.5, shapeName=dyn_jacket_bk_r_capsule, center=<5 0 0>, rotation=<rot 90 0 0>}]
 dyn_jacket_fr_l [{
   shapeType=Capsule
   length=9.003
   radius=2.8798
   shapeName=dyn_jacket_fr_l_capsule
   center=<4.6026 0.9013 4.7362>
   rotation=<rot -80.077 157.3228 -157.4446>
  }]
 dyn_jacket_fr_r [{
   shapeType=Capsule
   length=9.003
   radius=2.8798
   shapeName=dyn_jacket_fr_l_capsule
   center=<4.8698 0.3238 -4.6541>
   rotation=<rot -80.077 0 0>
  }]
 dyn_jacket_l [{shapeType=Capsule, length=7, radius=4.5, shapeName=dyn_jacket_l_capsule, center=<5 0 0>, rotation=<rot 90 0 0>}]
 dyn_jacket_r [{shapeType=Capsule, length=7, radius=4.5, shapeName=dyn_jacket_r_capsule, center=<5 0 0>, rotation=<rot 90 0 0>}]
 pelvis [{shapeType=Capsule, length=10, radius=10, shapeName=pelvis_capsule, center=<-1.5318 -0.3735 0>}]
 spine_01 [{shapeType=Capsule, length=4.5322, radius=9.9655, shapeName=spine_01_capsule, center=<6.6916 2.3429 0>, rotation=<rot 90 0 0>}]
 thigh_l [{shapeType=Capsule, length=29.0022, radius=8.9469, shapeName=thigh_l_capsule, center=<-20.1663 0.7704 0>, rotation=<rot 90 0 0>}]
 thigh_r [{shapeType=Capsule, length=29.0022, radius=8.9469, shapeName=thigh_l_capsule, center=<20.1663 0.7704 0>, rotation=<rot 90 0 0>}]
}
constraints [
 {
  bone1Name=base_jacket_bk
  bone2Name=spine_01
  swing1LimitDegrees=45
  swing1Motion=Free
  swing2LimitDegrees=45
  swing2Motion=Free
  twistLimitDegrees=45
  twistMotion=Free
 }
 {bone1Name=dyn_jacket_bk, bone2Name=base_jacket_bk, swing1LimitDegrees=10, swing2LimitDegrees=0, twistLimitDegrees=0}
 {
  bone1Name=base_jacket_bk_l
  bone2Name=spine_01
  swing1LimitDegrees=45
  swing1Motion=Free
  swing2LimitDegrees=45
  swing2Motion=Free
  twistLimitDegrees=45
  twistMotion=Free
 }
 {bone1Name=dyn_jacket_bk_l, bone2Name=base_jacket_bk_l, swing1LimitDegrees=10, swing2LimitDegrees=0, twistLimitDegrees=0}
 {
  bone1Name=base_jacket_bk_r
  bone2Name=spine_01
  swing1LimitDegrees=45
  swing1Motion=Free
  swing2LimitDegrees=45
  swing2Motion=Free
  twistLimitDegrees=45
  twistMotion=Free
 }
 {bone1Name=dyn_jacket_bk_r, bone2Name=base_jacket_bk_r, swing1LimitDegrees=10, swing2LimitDegrees=0, twistLimitDegrees=0}
 {
  bone1Name=base_jacket_fr_l
  bone2Name=spine_01
  swing1LimitDegrees=45
  swing1Motion=Free
  swing2LimitDegrees=45
  swing2Motion=Free
  twistLimitDegrees=45
  twistMotion=Free
 }
 {bone1Name=dyn_jacket_fr_l, bone2Name=base_jacket_fr_l, swing1LimitDegrees=25, swing2LimitDegrees=0, twistLimitDegrees=0}
 {
  bone1Name=base_jacket_fr_r
  bone2Name=spine_01
  swing1LimitDegrees=45
  swing1Motion=Free
  swing2LimitDegrees=45
  swing2Motion=Free
  twistLimitDegrees=45
  twistMotion=Free
 }
 {bone1Name=dyn_jacket_fr_r, bone2Name=base_jacket_fr_r, swing1LimitDegrees=25, swing2LimitDegrees=0, twistLimitDegrees=0}
 {
  bone1Name=base_jacket_l
  bone2Name=spine_01
  swing1LimitDegrees=45
  swing1Motion=Free
  swing2LimitDegrees=45
  swing2Motion=Free
  twistLimitDegrees=45
  twistMotion=Free
 }
 {bone1Name=dyn_jacket_l, bone2Name=base_jacket_l, swing1LimitDegrees=30, swing2LimitDegrees=0, twistLimitDegrees=0}
 {
  bone1Name=base_jacket_r
  bone2Name=spine_01
  swing1LimitDegrees=45
  swing1Motion=Free
  swing2LimitDegrees=45
  swing2Motion=Free
  twistLimitDegrees=45
  twistMotion=Free
 }
 {bone1Name=dyn_jacket_r, bone2Name=base_jacket_r, swing1LimitDegrees=30, swing2LimitDegrees=0, twistLimitDegrees=0}
 {
  bone1Name=spine_01
  bone2Name=pelvis
  swing1LimitDegrees=45
  swing1Motion=Free
  swing2LimitDegrees=45
  swing2Motion=Free
  twistLimitDegrees=45
  twistMotion=Free
 }
 {
  bone1Name=thigh_l
  bone2Name=pelvis
  swing1LimitDegrees=45
  swing1Motion=Free
  swing2LimitDegrees=45
  swing2Motion=Free
  twistLimitDegrees=45
  twistMotion=Free
 }
 {
  bone1Name=thigh_r
  bone2Name=pelvis
  swing1LimitDegrees=45
  swing1Motion=Free
  swing2LimitDegrees=45
  swing2Motion=Free
  twistLimitDegrees=45
  twistMotion=Free
 }
 {bone1Name=dyn_hip_laces_0_r, bone2Name=base_heart_jnt, swing1LimitDegrees=40, swing2LimitDegrees=20, twistLimitDegrees=0}
 {bone1Name=dyn_hip_laces_0_l, bone2Name=base_heart_jnt, swing1LimitDegrees=35, swing2LimitDegrees=20, twistLimitDegrees=0}
 {bone1Name=dyn_hip_laces_1_l, bone2Name=dyn_hip_laces_0_l, swing1LimitDegrees=20, swing2LimitDegrees=15, twistLimitDegrees=0}
 {bone1Name=dyn_hip_laces_1_r, bone2Name=dyn_hip_laces_0_r, swing1LimitDegrees=15, swing2LimitDegrees=20, twistLimitDegrees=0}
]
