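# Control Rig for the female medium BerryTartBrunt character body, adding procedural corrective motion to deform, jacket, heart and lace bones.
# A sequence runs debug-only control overrides, the shared DeformRig function, then outfit correctives.
# Jacket bones respond to low pelvis height (suppressed while IsSliding), upper-arm rotation and a head pose reader.
# Weighted parent constraints place heart and leg-lace bases, with the stronger thigh pose reader offsetting the heart.
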ControlRigBlueprint /BRCosmetics/Characters/Player/Female/Medium/Bodies/F_MED_BerryTartBrunt/Meshes/F_MED_BerryTartBrunt_CtrlRig
graphs [
 New Function_ContainedGraph {nodes [
   RigVMFunctionEntryNode()
   Remap = &RigVMFunction_MathDoubleRemap(
    SourceMaximum=45
    SourceMinimum=89
    TargetMaximum=1
    TargetMinimum=0
    Value=&RigUnit_GetRelativeTransformForItem(Child=(Type=Bone,Name="pelvis"), Parent=(Type=Bone,Name="root"), bChildInitial=False, bParentInitial=False).RelativeTransform.Translation.Z
   )
   > RigVMVariableNode(Value=@Remap.Result) {variable=lowToGround}
   DISPATCH_RigVMDispatch_If = RigVMDispatchNode(False=1, True=0, Condition=RigVMVariableNode() {variable=IsSliding}.Value) {template="DISPATCH_RigVMDispatch_If(in Condition,in True,in False,out Result)"}
   Multiply = &RigVMFunction_MathDoubleMul(A=RigVMVariableNode() {variable=lowToGround}.Value, B=@DISPATCH_RigVMDispatch_If.Result)
   > &RigUnit_ModifyTransforms(
    ItemToModify="((Item=(Type=Bone,Name=\"base_jacket_fr_l\"),Transform=(Rotation=(X=0.6086,Y=0.1394,Z=-0.0888,W=0.7761),Translation=(X=-3.792,Y=-2.458,Z=-2.331),Scale3D=(X=1,Y=1,Z=1))),(Item=(Type=Bone,Name=\"base_jacket_l\"),Transform=(Rotation=(X=0.228,Y=0.2028,Z=-0.2896,W= ...(truncated)"
    WeightMinimum=0
    Weight=@Multiply.Result
   )
   > &RigUnit_ParentConstraint(
    AdvancedSettings=(InterpolationType=Average,RotationOrderForFilter=XZY)
    Child=(Type=Bone,Name="upperarm_l_2_2")
    Parents=((Item=(Type=Bone,Name="upperarm_l"),Weight=1))
    Weight=1
   )
   n1 = &RigVMFunction_MathQuaternionToEuler(
    RotationOrder=ZXY
    Value=&RigUnit_GetRelativeTransformForItem(Child=(Type=Bone,Name="upperarm_l"), Parent=(Type=Bone,Name="clavicle_l"), bChildInitial=False, bParentInitial=False).RelativeTransform.Rotation
   )
   RigUnit_GetTransform = &RigUnit_GetTransform(Item=(Type=Bone,Name="upperarm_l_2_2"), Space=LocalSpace, bInitial=false)
   RigVMFunction_MathQuaternionSwingTwist = &RigVMFunction_MathQuaternionSwingTwist(TwistAxis=(X=1,Y=0,Z=0), Input=@RigUnit_GetTransform.Transform.Rotation)
   n2 = &RigVMFunction_MathDoubleClamp(
    Maximum=20
    Value=&RigVMFunction_MathQuaternionToEuler(RotationOrder=ZYX, Value=@RigVMFunction_MathQuaternionSwingTwist.Swing).Result.Z
   )
   > &RigUnit_ModifyTransforms(
    ItemToModify=((Item=(Type=Bone,Name="base_jacket_corrective_l"),Transform=(Rotation=(X=0,Y=0,Z=0,W=1),Translation=(X=2,Y=0,Z=-2),Scale3D=(X=1,Y=1,Z=1))))
    Weight=1
    WeightMinimum=0
    ItemToModify.0.Transform.Translation.Z=&RigVMFunction_MathDoubleRemap(SourceMaximum=80, SourceMinimum=0, TargetMaximum=-2, TargetMinimum=0, Value=@n1.Result.Z).Result
    ItemToModify.0.Transform.Translation.X=&RigVMFunction_MathDoubleRemap(SourceMaximum=80, SourceMinimum=0, TargetMaximum=2, TargetMinimum=0, Value=@n1.Result.Z).Result
    ItemToModify.0.Transform.Rotation=&RigVMFunction_MathQuaternionFromEuler(Euler=(X=0,Y=0,Z=0), Euler.Y=&RigVMFunction_MathDoubleMul(B=-1, A=@n2.Result).Result).Result
   )
   > &RigUnit_ParentConstraint(
    AdvancedSettings=(InterpolationType=Average,RotationOrderForFilter=XZY)
    Child=(Type=Bone,Name="upperarm_r_2_2")
    Parents=((Item=(Type=Bone,Name="upperarm_r"),Weight=1))
    Weight=1
   )
   n3 = &RigVMFunction_MathQuaternionToEuler(
    RotationOrder=ZYX
    Value=&RigVMFunction_MathQuaternionSwingTwist(
     TwistAxis=(X=1,Y=0,Z=0)
     Input=&RigUnit_GetTransform(Item=(Type=Bone,Name="upperarm_r_2_2"), Space=LocalSpace, bInitial=false).Transform.Rotation
    ).Swing
   )
   > &RigUnit_ModifyTransforms(
    ItemToModify=((Item=(Type=Bone,Name="clavicle_r_holder"),Transform=(Rotation=(X=0,Y=0,Z=0,W=1),Translation=(X=0,Y=0,Z=0),Scale3D=(X=1,Y=1,Z=1))))
    Weight=1
    WeightMinimum=0
    ItemToModify.0.Transform.Rotation=&RigVMFunction_MathQuaternionFromEuler(Euler=(X=0,Y=0,Z=0), Euler.Z=&RigVMFunction_MathDoubleClamp(Maximum=45, Value=@n3.Result.Z).Result).Result
   )
   > &RigUnit_ParentConstraint(
    AdvancedSettings=(InterpolationType=Average,RotationOrderForFilter=XZY)
    Child=(Type=Bone,Name="base_jacket_corrective_r")
    Parents=((Item=(Type=Bone,Name="clavicle_r_holder"),Weight=1))
    Weight=1
   )
   n4 = &RigVMFunction_MathQuaternionToEuler(
    RotationOrder=ZXY
    Value=&RigUnit_GetRelativeTransformForItem(Child=(Type=Bone,Name="upperarm_l"), Parent=(Type=Bone,Name="clavicle_l"), bChildInitial=False, bParentInitial=False).RelativeTransform.Rotation
   )
   > &RigUnit_ModifyTransforms(
    ItemToModify=((Item=(Type=Bone,Name="base_jacket_corrective_r"),Transform=(Rotation=(X=0.0872,Y=0,Z=0,W=0.9962),Translation=(X=2,Y=0,Z=-2),Scale3D=(X=1,Y=1,Z=1))))
    WeightMinimum=0
    Weight=&RigVMFunction_MathDoubleRemap(SourceMaximum=60, SourceMinimum=0, TargetMaximum=1, TargetMinimum=0, Value=@n3.Result.Z).Result
    ItemToModify.0.Transform.Translation.Z=&RigVMFunction_MathDoubleRemap(SourceMaximum=80, SourceMinimum=0, TargetMaximum=-2, TargetMinimum=0, Value=@n4.Result.Z).Result
    ItemToModify.0.Transform.Translation.X=&RigVMFunction_MathDoubleRemap(SourceMaximum=80, SourceMinimum=0, TargetMaximum=2, TargetMinimum=0, Value=@n4.Result.Z).Result
   )
   > n5 = &RigUnit_SphericalPoseReader(
    ActiveRegionScaleFactors=(PositiveWidth=1,NegativeWidth=1,PositiveHeight=1,NegativeHeight=1)
    ActiveRegionSize=0.01
    Debug=(bDrawDebug=false,bDraw2D=False,bDrawLocalAxes=False,DebugScale=25,DebugSegments=20,DebugThickness=0.25)
    DriverAxis=(X=0,Y=-1,Z=0)
    DriverItem=(Type=Bone,Name="head")
    FalloffRegionScaleFactors=(PositiveWidth=1,NegativeWidth=1,PositiveHeight=1,NegativeHeight=1)
    FalloffSize=0.34
    FlipHeightScaling=False
    FlipWidthScaling=False
    OptionalParentItem=(Type=Bone,Name="neck_01")
    RotationOffset=(X=0,Y=80,Z=30)
   )
   RigVMFunction_AnimEvalRichCurve = &RigVMFunction_AnimEvalRichCurve(
    Curve=(EditorCurveData=(Keys=((InterpMode=RCIM_Cubic),(InterpMode=RCIM_Cubic,Time=1,Value=1))))
    SourceMinimum=0
    TargetMinimum=0
    Value=@n5.OutputParam
   )
   > &RigUnit_ModifyTransforms(
    ItemToModify=((Item=(Type=Bone,Name="dyn_corrective_jacket"),Transform=(Rotation=(X=-0.0034,Y=0.1787,Z=0.2473,W=0.9523),Translation=(X=0,Y=0,Z=-3),Scale3D=(X=1,Y=1,Z=1.4))))
    WeightMinimum=0
    Weight=@RigVMFunction_AnimEvalRichCurve.Result
   )
   > RigVMFunctionReturnNode()
  ]}
 New Function_ContainedGraph {nodes [
   n1 = RigVMFunctionEntryNode()
   RigVMFunction_ControlFlowBranch = &RigVMFunction_ControlFlowBranch(ExecuteContext=@n1.ExecuteContext, Condition=@n1.debug)
   RigVMFunctionReturnNode(ExecuteContext=@RigVMFunction_ControlFlowBranch.Completed)
   RigUnit_GetTransform = &RigUnit_GetTransform(Item=(Type=Control,Name="thigh_l_ctrl"), bInitial=False)
   RigUnit_SetTransform = &RigUnit_SetTransform(
    Item=(Type=Bone,Name="thigh_l")
    Weight=1
    bInitial=False
    Value=@RigUnit_GetTransform.Transform
    ExecutePin=@RigVMFunction_ControlFlowBranch.True
   )
   > &RigUnit_SetTransform(
    Item=(Type=Bone,Name="thigh_r")
    Weight=1
    bInitial=False
    Value=&RigUnit_GetTransform(Item=(Type=Control,Name="thigh_r_ctrl"), bInitial=False).Transform
   )
   > &RigUnit_SetTransform(
    Item=(Type=Bone,Name="clavicle_l")
    Weight=1
    bInitial=False
    Value=&RigUnit_GetTransform(Item=(Type=Control,Name="clavicle_l_ctrl"), bInitial=False).Transform
   )
   > &RigUnit_SetTransform(
    Item=(Type=Bone,Name="clavicle_r")
    Weight=1
    bInitial=False
    Value=&RigUnit_GetTransform(Item=(Type=Control,Name="clavicle_r_ctrl"), bInitial=False).Transform
   )
   > &RigUnit_SetTransform(
    Item=(Type=Bone,Name="upperarm_l")
    Weight=1
    bInitial=False
    Value=&RigUnit_GetTransform(Item=(Type=Control,Name="upperarm_l_ctrl"), bInitial=False).Transform
   )
   > &RigUnit_SetTransform(
    Item=(Type=Bone,Name="upperarm_r")
    Weight=1
    bInitial=False
    Value=&RigUnit_GetTransform(Item=(Type=Control,Name="upperarm_r_ctrl"), bInitial=False).Transform
   )
   > &RigUnit_SetTransform(
    Item=(Type=Bone,Name="head")
    Weight=1
    bInitial=False
    Value=&RigUnit_GetTransform(Item=(Type=Control,Name="head_ctrl"), bInitial=False).Transform
   )
  ]}
 RigVMFunctionLibrary {nodes [
   controls_FNC = RigVMCollapseNode(ExecuteContext=(), debug=False)
   jacket_FNC = RigVMCollapseNode(ExecuteContext=())
  ]}
 RigVMFunction_Sequence_1_ContainedGraph {nodes [
   n1 = RigVMFunctionEntryNode()
   RigVMFunction_Sequence = &RigVMFunction_Sequence(ExecuteContext=@n1.ExecuteContext)
   n2 = &RigVMFunction_Sequence(ExecuteContext=@RigVMFunction_Sequence.B)
   RigVMFunctionReturnNode(A=@RigVMFunction_Sequence.A, B=@n2.A, C=@n2.B)
  ]}
 RigVMModel {nodes [
   RigUnit_BeginExecution = &RigUnit_BeginExecution()
   RigVMFunction_Sequence = RigVMAggregateNode(ExecuteContext=@RigUnit_BeginExecution.ExecutePin)
   controls_FNC = RigVMFunctionReferenceNode(debug=false, ExecuteContext=@RigVMFunction_Sequence.A)
   FN_DeformRig_FNC = RigVMFunctionReferenceNode(
    DeformRig_ Curve="(deformRig_l_upperarm_up_alpha_72_#1=False,deformRig_r_upperarm_up_alpha_73_#2=False,deformRig_l_upperarm_fr_alpha_74_#3=False,deformRig_r_upperarm_fr_alpha_75_CB51B8 ...(truncated)"
    DeformRig_Bool="(upperarm_1_#4=True,elbow_3_#5=True,wrist_7_#6=True,groin_8_#7=True,glute_17_#8=True,knee_9_40C408324CA261 ...(truncated)"
    DeformRig_Debug=(upperarm_debug_18_#4=False,wrist_debug_20_#6=False,lat_debug_26_#9=False,groinGlute_debug_27_#7=false)
    DeformRig_Value="(upperarm_up_18_#10=(Rotation=(X=0.0,Y=-0.0,Z=0.0,W=1.0),Translation=(X=0.0,Y=0.0,Z=8.0),Scale3D=(X=0.70,Y=1.0,Z=1.0)),upperarm_fwd_fr_19_#11=(Rotation=(X=0.0,Y=-0.0,Z=0.0,W=1.0),Translation=(X=0.0, ...(truncated)"
    ExecuteContext=@RigVMFunction_Sequence.B
   )
   jacket_FNC = RigVMFunctionReferenceNode(ExecuteContext=@RigVMFunction_Sequence.C)
   > &RigUnit_ParentConstraint(
    AdvancedSettings=(RotationOrderForFilter=XZY)
    Child=(Type=Bone,Name="base_heart_jnt")
    Parents="((Item=(Type=Bone,Name=\"thigh_r\"),Weight=0.0001),(Item=(Type=Bone,Name=\"spine_03\"),Weight=0.0003),(Item=(Type=Bone,Name=\"thigh_twist_01_l\"),Weight=0.0308),(Item=(Type=Bone,Name=\"thigh_twist_01_r\"),Weight=0.0308),(Item=(Type=Bone,Name=\"spine_02\"),Weight=0.0 ...(truncated)"
    Weight=1
   )
   > &RigUnit_ParentConstraint(
    AdvancedSettings=(RotationOrderForFilter=XZY)
    Child=(Type=Bone,Name="base_leg_laces_l")
    Parents=((Item=(Type=Bone,Name="deform_groin_l"),Weight=0.0144),(Item=(Type=Bone,Name="pelvis"),Weight=0.0273),(Item=(Type=Bone,Name="thigh_twist_01_l"),Weight=0.4148),(Item=(Type=Bone,Name="thigh_l"),Weight=0.5435))
    Weight=1
   )
   > &RigUnit_ParentConstraint(
    AdvancedSettings=(RotationOrderForFilter=XZY)
    Child=(Type=Bone,Name="base_leg_laces_r")
    Parents=((Item=(Type=Bone,Name="deform_groin_r"),Weight=0.0144),(Item=(Type=Bone,Name="pelvis"),Weight=0.0273),(Item=(Type=Bone,Name="thigh_twist_01_r"),Weight=0.4148),(Item=(Type=Bone,Name="thigh_r"),Weight=0.5435))
    Weight=1
   )
   > n1 = &RigUnit_SphericalPoseReader(
    ActiveRegionScaleFactors=(PositiveWidth=1,NegativeWidth=1,PositiveHeight=1,NegativeHeight=1)
    ActiveRegionSize=0.1
    Debug=(bDrawDebug=false,bDraw2D=False,bDrawLocalAxes=False,DebugScale=25,DebugSegments=20,DebugThickness=0.25)
    DriverAxis=(X=-1,Y=0,Z=0)
    DriverItem=(Type=Bone,Name="thigh_l")
    FalloffRegionScaleFactors=(PositiveWidth=1,NegativeWidth=1.3,PositiveHeight=1,NegativeHeight=1)
    FalloffSize=0.2
    FlipHeightScaling=False
    FlipWidthScaling=False
    OptionalParentItem=(Type=Bone,Name="None")
    RotationOffset=(X=90,Y=0,Z=0)
   )
   > n2 = &RigUnit_SphericalPoseReader(
    ActiveRegionScaleFactors=(PositiveWidth=1,NegativeWidth=1,PositiveHeight=1,NegativeHeight=1)
    ActiveRegionSize=0.1
    Debug=(bDrawDebug=false,bDraw2D=False,bDrawLocalAxes=False,DebugScale=25,DebugSegments=20,DebugThickness=0.25)
    DriverAxis=(X=1,Y=0,Z=0)
    DriverItem=(Type=Bone,Name="thigh_r")
    FalloffRegionScaleFactors=(PositiveWidth=1.8,NegativeWidth=1,PositiveHeight=1,NegativeHeight=1)
    FalloffSize=0.2
    FlipHeightScaling=False
    FlipWidthScaling=False
    OptionalParentItem=(Type=Bone,Name="None")
    RotationOffset=(X=-90,Y=0,Z=0)
   )
   GreaterEqual = /Script/RigVM.RigVMFunction_MathDoubleGreaterEqual(B=@n2.OutputParam, A=@n1.OutputParam) {is=struct RigVMUnitNode, template="GreaterEqual::Execute(in A,in B,out Result)"}
   DISPATCH_RigVMDispatch_If = RigVMDispatchNode(Condition=@GreaterEqual.Result, False=@n2.OutputParam, True=@n1.OutputParam) {template="DISPATCH_RigVMDispatch_If(in Condition,in True,in False,out Result)"}
   > &RigUnit_ModifyTransforms(
    ItemToModify=((Item=(Type=Bone,Name="base_heart_jnt"),Transform=(Rotation=(X=0,Y=0,Z=0.1259,W=0.992),Translation=(X=-0.048,Y=-0.5791,Z=0),Scale3D=(X=1,Y=1,Z=1))))
    WeightMinimum=0
    Weight=@DISPATCH_RigVMDispatch_If.Result
   )
  ]}
]
hierarchy [
 root {loc=<0 0 0>, rot=<0 0 0 1>}
 pelvis {parent=root, loc=<0 -0.5301 98.0571>, rot=<0.0127 0.707 -0.0127 -0.707>}
 spine_01 {parent=pelvis, loc=<6.1208 -0.9328 0>, rot=<0 0 0.034 -0.9994>}
 spine_02 {parent=spine_01, loc=<6.5997 -0.2095 0>, rot=<0 0 0.0238 -0.9997>}
 spine_03 {parent=spine_02, loc=<6.4914 -0.1237 0>, rot=<0 0 0.0149 -0.9999>}
 spine_04 {parent=spine_03, loc=<6.9598 0.0794 0>, rot=<0 0 -0.0132 -0.9999>}
 spine_05 {parent=spine_04, loc=<6.9987 0 0>, rot=<0 0 -0.0015 -1>}
 attach_cape {parent=spine_05, loc=<9.0177 -11.7138 0>, rot=<0.0616 -0.7044 -0.0616 -0.7044>}
 neck_01 {parent=spine_05, loc=<11.3138 0.1529 0>, rot=<0 0 -0.2003 -0.9797>}
 neck_02 {parent=neck_01, loc=<4.5627 0 0>, rot=<0 0 0.0504 -0.9987>}
 head {parent=neck_02, loc=<5.3914 0 0>, rot=<0 0 0.115 -0.9934>}
 attach_backpack {parent=spine_05, loc=<-0.1222 -13.4728 0>, rot=<-0.0093 -0.707 0.0093 -0.707>}
 clavicle_l {parent=spine_05, loc=<5.4831 -0.1053 -3.2946>, rot=<0.0465 -0.7289 -0.009 -0.6829>}
 upperarm_l {parent=clavicle_l, loc=<9.8658 -1.2136 -0.4105>, rot=<0.0164 -0.3245 -0.0269 -0.9454>}
 lowerarm_l {parent=upperarm_l, loc=<26.465 -0.8345 0.259>, rot=<0.0036 0.0161 -0.2298 -0.9731>}
 hand_l {parent=lowerarm_l, loc=<22.7355 -0.3203 -0.6105>, rot=<-0.6784 -0.0223 0.0735 -0.7307>}
 thumb_01_l {parent=hand_l, loc=<1.3118 -0.3191 -2.459>, rot=<0.6546 -0.4042 0.0664 -0.6354>}
 thumb_02_l {parent=thumb_01_l, loc=<5.9336 -0.0448 0.0401>, rot=<-0.0037 0.0213 0.298 -0.9543>}
 thumb_03_l {parent=thumb_02_l, loc=<2.2223 0.3682 -0.0907>, rot=<0.0118 -0.013 0.0066 -0.9998>}
 index_metacarpal_l {parent=hand_l, loc=<1.6535 0.0183 -2.96>, rot=<0.0158 -0.0124 -0.0173 -0.9996>}
 index_01_l {parent=index_metacarpal_l, loc=<8.3008 0.0357 -0.0524>, rot=<-0.0044 0.0485 0.0924 -0.9945>}
 index_02_l {parent=index_01_l, loc=<3.3587 -0.3004 0.0452>, rot=<-0.002 0.0027 0.3225 -0.9466>}
 index_03_l {parent=index_02_l, loc=<2.5578 -0.123 -0.0252>, rot=<-0.002 0.0067 0.2558 -0.9667>}
 middle_metacarpal_l {parent=hand_l, loc=<2.1856 -0.1725 -0.5984>, rot=<0.006 0.0168 -0.0329 -0.9993>}
 middle_01_l {parent=middle_metacarpal_l, loc=<7.3275 -0.2447 0.0129>, rot=<-0.0208 0.0042 0.0858 -0.9961>}
 middle_02_l {parent=middle_01_l, loc=<4.0123 0.1401 0.1175>, rot=<-0.0218 0.0258 0.3512 -0.9357>}
 middle_03_l {parent=middle_02_l, loc=<2.6077 0.0534 0.1373>, rot=<0.0082 0.0188 0.1708 -0.9851>}
 ring_metacarpal_l {parent=hand_l, loc=<2.2074 -0.3625 1.3638>, rot=<-0.0284 0.0722 -0.0458 -0.9959>}
 ring_01_l {parent=ring_metacarpal_l, loc=<7.0295 -0.6197 -0.1446>, rot=<-0.0489 -0.0252 0.1892 -0.9804>}
 ring_02_l {parent=ring_01_l, loc=<3.6174 0.4647 -0.1027>, rot=<0 0 0.2806 -0.9598>}
 ring_03_l {parent=ring_02_l, loc=<2.4152 0.114 0.0199>, rot=<0 0 0.1189 -0.9929>}
 pinky_metacarpal_l {parent=hand_l, loc=<1.7616 -0.9753 3.3814>, rot=<-0.0397 0.0814 -0.069 -0.9935>}
 pinky_01_l {parent=pinky_metacarpal_l, loc=<7.1709 -0.4831 -0.1094>, rot=<-0.0654 -0.0035 0.2242 -0.9723>}
 pinky_02_l {parent=pinky_01_l, loc=<2.8274 0.0893 -0.1043>, rot=<0.0037 0.0011 0.2854 -0.9584>}
 pinky_03_l {parent=pinky_02_l, loc=<2.2046 0.0307 -0.0078>, rot=<-0.0002 0.0098 0.1445 -0.9895>}
 deform_wrist_dn_base_l {parent=hand_l, loc=<0 -1.5326 0>, rot=<0 0 0 -1>}
 deform_wrist_dn_l {parent=deform_wrist_dn_base_l, loc=<0 0 0>, rot=<0 0 0 1>}
 deform_wrist_up_base_l {parent=hand_l, loc=<0 2.3009 0>, rot=<0 0 0 -1>}
 deform_wrist_up_l {parent=deform_wrist_up_base_l, loc=<0 0 0>, rot=<0 0 0 1>}
 lowerarm_twist_02_l {parent=lowerarm_l, loc=<10.6707 0 0>, rot=<0 0 0 1>}
 lowerarm_twist_01_l {parent=lowerarm_l, loc=<18.316 0 0>, rot=<0 0 0 1>}
 deform_elbow_in_base_l {parent=lowerarm_l, loc=<0.1834 3.6062 -0.2572>, rot=<0 0 0 -1>}
 deform_elbow_in_l {parent=deform_elbow_in_base_l, loc=<0 0 0>, rot=<0 0 0 1>}
 deform_elbow_out_base_l {parent=lowerarm_l, loc=<0 -4 0>, rot=<0 0 0 1>}
 deform_elbow_out_l {parent=deform_elbow_out_base_l, loc=<0 0 0>, rot=<0 0 0 1>}
 upperarm_twist_01_l {parent=upperarm_l, loc=<7.2135 -0.1134 0>, rot=<0 0 0 1>}
 deform_upperarm_bk_base_l {parent=upperarm_twist_01_l, loc=<-6.4053 -4.3746 0.314>, rot=<0 0 0 -1>}
 deform_upperarm_bk_l {parent=deform_upperarm_bk_base_l, loc=<0 0 0>, rot=<0 0 0 1>}
 deform_upperarm_fr_base_l {parent=upperarm_twist_01_l, loc=<-5.6312 5.092 0.3354>, rot=<0 0 0 -1>}
 deform_upperarm_fr_l {parent=deform_upperarm_fr_base_l, loc=<0 0 0>, rot=<0 0 0 1>}
 deform_upperarm_up_base_l {parent=upperarm_twist_01_l, loc=<-8.891 0.4825 3.9517>, rot=<0 0 0 -1>}
 deform_upperarm_up_l {parent=deform_upperarm_up_base_l, loc=<0 0 0>, rot=<0 0 0 1>}
 upperarm_twist_02_l {parent=upperarm_l, loc=<19.7998 -0.1134 0>, rot=<0 0 0 1>}
 upperarm_l_2 {parent=clavicle_l, loc=<9.8658 -1.2136 -0.4105>, rot=<0.0164 -0.3245 -0.0269 -0.9454>}
 upperarm_l_2_2 {parent=upperarm_l_2, loc=<0 0 0>, rot=<0 0 0 1>}
 clavicle_r {parent=spine_05, loc=<5.4831 -0.1053 3.2946>, rot=<0.7289 0.0465 -0.6829 0.009>}
 upperarm_r {parent=clavicle_r, loc=<-9.8658 1.2136 0.4105>, rot=<0.0164 -0.3245 -0.0269 -0.9454>}
 lowerarm_r {parent=upperarm_r, loc=<-26.465 0.8345 -0.259>, rot=<0.0036 0.0161 -0.2298 -0.9731>}
 hand_r {parent=lowerarm_r, loc=<-22.7355 0.3203 0.6106>, rot=<-0.6784 -0.0223 0.0735 -0.7307>}
 thumb_01_r {parent=hand_r, loc=<-1.3118 0.3191 2.459>, rot=<0.6546 -0.4042 0.0664 -0.6354>}
 thumb_02_r {parent=thumb_01_r, loc=<-5.9336 0.0448 -0.0401>, rot=<-0.0037 0.0213 0.298 -0.9543>}
 thumb_03_r {parent=thumb_02_r, loc=<-2.2223 -0.3682 0.0907>, rot=<0.0118 -0.013 0.0066 -0.9998>}
 index_metacarpal_r {parent=hand_r, loc=<-1.6535 -0.0183 2.96>, rot=<0.0158 -0.0124 -0.0173 -0.9996>}
 index_01_r {parent=index_metacarpal_r, loc=<-8.3008 -0.0357 0.0524>, rot=<-0.0044 0.0485 0.0924 -0.9945>}
 index_02_r {parent=index_01_r, loc=<-3.3587 0.3004 -0.0452>, rot=<-0.002 0.0027 0.3225 -0.9466>}
 index_03_r {parent=index_02_r, loc=<-2.5578 0.123 0.0252>, rot=<-0.002 0.0067 0.2558 -0.9667>}
 middle_metacarpal_r {parent=hand_r, loc=<-2.1856 0.1725 0.5984>, rot=<0.006 0.0168 -0.0329 -0.9993>}
 middle_01_r {parent=middle_metacarpal_r, loc=<-7.3275 0.2447 -0.0129>, rot=<-0.0208 0.0042 0.0858 -0.9961>}
 middle_02_r {parent=middle_01_r, loc=<-4.0123 -0.1401 -0.1175>, rot=<-0.0218 0.0258 0.3512 -0.9357>}
 middle_03_r {parent=middle_02_r, loc=<-2.6077 -0.0534 -0.1373>, rot=<0.0082 0.0188 0.1708 -0.9851>}
 ring_metacarpal_r {parent=hand_r, loc=<-2.2074 0.3625 -1.3638>, rot=<-0.0284 0.0722 -0.0458 -0.9959>}
 ring_01_r {parent=ring_metacarpal_r, loc=<-7.0295 0.6197 0.1446>, rot=<-0.0489 -0.0252 0.1892 -0.9804>}
 ring_02_r {parent=ring_01_r, loc=<-3.6174 -0.4647 0.1027>, rot=<0 0 0.2806 -0.9598>}
 ring_03_r {parent=ring_02_r, loc=<-2.4152 -0.114 -0.0199>, rot=<0 0 0.1189 -0.9929>}
 pinky_metacarpal_r {parent=hand_r, loc=<-1.7616 0.9753 -3.3814>, rot=<-0.0397 0.0814 -0.069 -0.9935>}
 pinky_01_r {parent=pinky_metacarpal_r, loc=<-7.1709 0.4831 0.1094>, rot=<-0.0654 -0.0035 0.2242 -0.9723>}
 pinky_02_r {parent=pinky_01_r, loc=<-2.8274 -0.0893 0.1043>, rot=<0.0037 0.0011 0.2854 -0.9584>}
 pinky_03_r {parent=pinky_02_r, loc=<-2.2046 -0.0307 0.0078>, rot=<-0.0002 0.0098 0.1445 -0.9895>}
 deform_wrist_dn_base_r {parent=hand_r, loc=<-0.0001 1.5328 0>, rot=<0 0 0 -1>}
 deform_wrist_dn_r {parent=deform_wrist_dn_base_r, loc=<0 0 0>, rot=<0 0 0 1>}
 deform_wrist_up_base_r {parent=hand_r, loc=<-0.0002 -2.3007 -0.0001>, rot=<0 0 0 -1>}
 deform_wrist_up_r {parent=deform_wrist_up_base_r, loc=<0 0 0>, rot=<0 0 0 1>}
 lowerarm_twist_02_r {parent=lowerarm_r, loc=<-10.6707 0 0>, rot=<0 0 0 -1>}
 lowerarm_twist_01_r {parent=lowerarm_r, loc=<-18.316 0 0>, rot=<0 0 0 -1>}
 deform_elbow_in_base_r {parent=lowerarm_r, loc=<-0.1834 -3.6062 0.2572>, rot=<0 0 0 -1>}
 deform_elbow_in_r {parent=deform_elbow_in_base_r, loc=<0 0 0>, rot=<0 0 0 1>}
 deform_elbow_out_base_r {parent=lowerarm_r, loc=<0 4 -0.0001>, rot=<0 0 0 -1>}
 deform_elbow_out_r {parent=deform_elbow_out_base_r, loc=<0 0 0>, rot=<0 0 0 1>}
 upperarm_twist_01_r {parent=upperarm_r, loc=<-7.2135 0.1134 0>, rot=<0 0 0 -1>}
 deform_upperarm_bk_base_r {parent=upperarm_twist_01_r, loc=<6.4053 4.3746 -0.314>, rot=<0 0 0 -1>}
 deform_upperarm_bk_r {parent=deform_upperarm_bk_base_r, loc=<0 0 0>, rot=<0 0 0 1>}
 deform_upperarm_fr_base_r {parent=upperarm_twist_01_r, loc=<5.6312 -5.092 -0.3354>, rot=<0 0 0 -1>}
 deform_upperarm_fr_r {parent=deform_upperarm_fr_base_r, loc=<0 0 0>, rot=<0 0 0 1>}
 deform_upperarm_up_base_r {parent=upperarm_twist_01_r, loc=<8.891 -0.4825 -3.9517>, rot=<0 0 0 -1>}
 deform_upperarm_up_r {parent=deform_upperarm_up_base_r, loc=<0 0 0>, rot=<0 0 0 1>}
 upperarm_twist_02_r {parent=upperarm_r, loc=<-19.7998 0.1134 0>, rot=<0 0 0 -1>}
 upperarm_r_2 {parent=clavicle_r, loc=<-9.8658 1.2136 0.4105>, rot=<0.0164 -0.3245 -0.0269 -0.9454>}
 upperarm_r_2_2 {parent=upperarm_r_2, loc=<0 0 0>, rot=<0 0 0 1>}
 base_necklace_00 {parent=spine_05, loc=<9.6374 1.8755 -0.0165>, rot=<0 0 0.8856 0.4645>}
 dyn_necklace_00 {parent=base_necklace_00, loc=<0 0 0>, rot=<0 0 0 1>}
 dyn_necklace_01 {parent=dyn_necklace_00, loc=<3 0 0>, rot=<0 0 -0.1471 -0.9891>}
 dyn_necklace_02 {parent=dyn_necklace_01, loc=<3 0 0>, rot=<0 0 -0.1045 -0.9945>}
 base_jacket_corrective_r {parent=spine_05, loc=<4.552 -1.892 13.037>, rot=<-0.6539 -0.7497 -0.0714 0.0731>}
 dyn_jacket_corrective_0_r {parent=base_jacket_corrective_r, loc=<0 0 0>, rot=<0 0 0 -1>}
 base_jacket_corrective_l {parent=spine_05, loc=<4.552 -1.892 -13.037>, rot=<0.0731 -0.0714 -0.7497 -0.6539>}
 dyn_jacket_corrective_0_l {parent=base_jacket_corrective_l, loc=<0 0 0>, rot=<0 0 0 -1>}
 base_corrective_jacket {parent=spine_05, loc=<11.3138 0.1529 0>, rot=<0.8344 -0.5512 0 0>}
 dyn_corrective_jacket {parent=base_corrective_jacket, loc=<0 0 0>, rot=<0 0 0 1>}
 clavicle_r_holder {parent=spine_05, loc=<5.4831 -0.1053 3.2946>, rot=<0.7289 0.0465 -0.6829 0.009>}
 deform_pec_aim_base_l {parent=spine_04, loc=<3.2456 9.3309 -1.2237>, rot=<0.0603 -0.4033 0.3318 -0.8507>}
 deform_pec_aim_l {parent=deform_pec_aim_base_l, loc=<0 0 0>, rot=<0 0 0 1>}
 deform_pec_base_l {parent=deform_pec_aim_l, loc=<9.3665 0 0>, rot=<0 0 0 1>}
 deform_pec_l {parent=deform_pec_base_l, loc=<0 0 0>, rot=<0 0 0 1>}
 deform_pec_aim_base_r {parent=spine_04, loc=<3.2452 9.3309 1.2237>, rot=<-0.4033 -0.0603 0.8507 0.3318>}
 deform_pec_aim_r {parent=deform_pec_aim_base_r, loc=<0 0 0>, rot=<0 0 0 1>}
 deform_pec_base_r {parent=deform_pec_aim_r, loc=<-9.3667 -0.0003 -0.0004>, rot=<0 0 0 1>}
 deform_pec_r {parent=deform_pec_base_r, loc=<0 0 0>, rot=<0 0 0 1>}
 base_jacket_fr_l {parent=spine_03, loc=<-1.0043 5.2923 -14.9357>, rot=<0.0387 -0.706 -0.706 0.0387>}
 dyn_jacket_fr_l {parent=base_jacket_fr_l, loc=<0 0 0>, rot=<0 0 0 1>}
 base_jacket_bk {parent=spine_03, loc=<4.0364 -13.0966 0>, rot=<0.0547 -0.9985 0 0>}
 dyn_jacket_bk {parent=base_jacket_bk, loc=<0 0 0>, rot=<0 0 0 1>}
 base_jacket_fr_r {parent=spine_03, loc=<-1.0043 5.2923 14.9357>, rot=<0.0387 -0.706 0.706 -0.0387>}
 dyn_jacket_fr_r {parent=base_jacket_fr_r, loc=<0 0 0>, rot=<0 0 0 1>}
 base_jacket_l {parent=spine_03, loc=<2.9723 -3.4208 -14.4773>, rot=<0.0419 -0.7649 -0.6418 0.0352>}
 dyn_jacket_l {parent=base_jacket_l, loc=<0 0 0>, rot=<0 0 0 1>}
 base_jacket_bk_l {parent=spine_03, loc=<3.8322 -11.2403 -8.9956>, rot=<0.0529 -0.9645 -0.2584 0.0142>}
 dyn_jacket_bk_l {parent=base_jacket_bk_l, loc=<0 0 0>, rot=<0 0 0 1>}
 base_jacket_bk_r {parent=spine_03, loc=<3.8318 -11.2404 8.9956>, rot=<0.0529 -0.9645 0.2584 -0.0142>}
 dyn_jacket_bk_r {parent=base_jacket_bk_r, loc=<0 0 0>, rot=<0 0 0 1>}
 base_jacket_r {parent=spine_03, loc=<2.9719 -3.4208 14.4773>, rot=<0.0419 -0.7649 0.6418 -0.0352>}
 dyn_jacket_r {parent=base_jacket_r, loc=<0 0 0>, rot=<0 0 0 -1>}
 thigh_l {parent=pelvis, loc=<-1.0855 -0.2109 -8.7053>, rot=<0.0009 0.0355 0.0155 -0.9993>}
 calf_l {parent=thigh_l, loc=<-42.5284 -0.0019 -0.7297>, rot=<0.0006 -0.0161 -0.0391 -0.9991>}
 foot_l {parent=calf_l, loc=<-46.7401 -2.5014 -0.592>, rot=<0.0427 -0.0238 0.0353 -0.9982>}
 ball_l {parent=foot_l, loc=<-3.5658 10.8066 -0.1544>, rot=<0.0001 -0.0001 -0.7063 -0.7079>}
 calf_twist_02_l {parent=calf_l, loc=<-19.1546 -1.7943 -0.2638>, rot=<0.0237 0.0095 -0.0035 -0.9997>}
 calf_twist_01_l {parent=calf_l, loc=<-35.8483 -1.9197 -0.5794>, rot=<0.0237 0.0095 -0.0035 -0.9997>}
 deform_knee_in_base_l {parent=calf_l, loc=<0.8031 -5.0231 0.2456>, rot=<0 0 0 -1>}
 deform_knee_in_l {parent=deform_knee_in_base_l, loc=<0 0 0>, rot=<0 0 0 1>}
 deform_knee_out_base_l {parent=calf_l, loc=<0 4 0>, rot=<0 0 0 -1>}
 deform_knee_out_l {parent=deform_knee_out_base_l, loc=<0 0 0>, rot=<0 0 0 1>}
 base_calf_front_laces_l {parent=calf_l, loc=<-24.5102 4.4476 -1.552>, rot=<-0.0242 -0.0001 0.9994 0.024>}
 dyn_calf_front_laces_l {parent=base_calf_front_laces_l, loc=<0 0 0>, rot=<0 0 0 1>}
 base_calf_back_laces_l {parent=calf_l, loc=<-26.4551 -11.3038 -0.13>, rot=<-0.0317 -0.9995 -0.0009 -0.0053>}
 dyn_calf_back_laces_l {parent=base_calf_back_laces_l, loc=<0 0 0>, rot=<0 0 0 1>}
 thigh_twist_01_l {parent=thigh_l, loc=<-27.952 0 0>, rot=<0.0013 0 0 -1>}
 deform_groin_base_l {parent=thigh_l, loc=<0.4801 5.6776 -0.2486>, rot=<0 0 0 -1>}
 deform_groin_l {parent=deform_groin_base_l, loc=<0 0 0>, rot=<0 0 0 1>}
 base_leg_laces_l {parent=thigh_l, loc=<-16.581 8.332 -6.693>, rot=<0.0349 0.2584 0.9654 -0.0068>}
 dyn_leg_laces_l {parent=base_leg_laces_l, loc=<0 0 0>, rot=<0 0 0 1>}
 dyn_leg_laces_0_l {parent=dyn_leg_laces_l, loc=<0.299 0.3731 -1.102>, rot=<-0.6579 -0.3761 0.3322 -0.5615>}
 dyn_leg_laces_1_l {parent=dyn_leg_laces_l, loc=<0.229 -0.3799 1.141>, rot=<0.6118 0.2641 0.3369 -0.6652>}
 dyn_leg_laces_2_l {parent=dyn_leg_laces_l, loc=<1.424 -0.3822 -1.0101>, rot=<0.0006 -0.0396 -0.0351 -0.9986>}
 dyn_leg_laces_3_l {parent=dyn_leg_laces_2_l, loc=<4 0 0>, rot=<-0.0004 0.0266 0.0146 -0.9995>}
 dyn_leg_laces_4_l {parent=dyn_leg_laces_3_l, loc=<4.2784 0 0>, rot=<0 0 0 1>}
 dyn_leg_laces_5_l {parent=dyn_leg_laces_4_l, loc=<4.3664 0 0>, rot=<0 0 0 1>}
 dyn_leg_laces_6_l {parent=dyn_leg_laces_l, loc=<1.4443 -0.4006 0.6461>, rot=<0.9998 -0.0214 0 0.0004>}
 dyn_leg_laces_7_l {parent=dyn_leg_laces_6_l, loc=<2.6439 0.005 0>, rot=<0 0 0.0223 -0.9998>}
 dyn_leg_laces_8_l {parent=dyn_leg_laces_7_l, loc=<2.6227 0 0>, rot=<0 0.0104 0 -0.9999>}
 dyn_leg_laces_9_l {parent=dyn_leg_laces_8_l, loc=<3.0037 0 0>, rot=<0 0 0 1>}
 thigh_r {parent=pelvis, loc=<-1.0855 -0.2109 8.7053>, rot=<0.0355 -0.0009 0.9993 0.0155>}
 calf_r {parent=thigh_r, loc=<42.5284 0.0019 0.7297>, rot=<0.0006 -0.0161 -0.0391 -0.9991>}
 foot_r {parent=calf_r, loc=<46.7401 2.5014 0.592>, rot=<0.0427 -0.0238 0.0353 -0.9982>}
 ball_r {parent=foot_r, loc=<3.5658 -10.8066 0.1544>, rot=<0.0001 -0.0001 -0.7063 -0.7079>}
 calf_twist_02_r {parent=calf_r, loc=<19.1546 1.7943 0.2638>, rot=<0.0237 0.0095 -0.0035 -0.9997>}
 calf_twist_01_r {parent=calf_r, loc=<35.8483 1.9197 0.5794>, rot=<0.0237 0.0095 -0.0035 -0.9997>}
 deform_knee_in_base_r {parent=calf_r, loc=<-0.8031 5.0231 -0.2457>, rot=<0 0 0 -1>}
 deform_knee_in_r {parent=deform_knee_in_base_r, loc=<0 0 0>, rot=<0 0 0 1>}
 deform_knee_out_base_r {parent=calf_r, loc=<0 -4 0>, rot=<0 0 0 -1>}
 deform_knee_out_r {parent=deform_knee_out_base_r, loc=<0 0 0>, rot=<0 0 0 1>}
 base_calf_front_laces_r {parent=calf_r, loc=<24.5102 -4.4476 1.552>, rot=<0.9994 -0.024 0.0242 -0.0001>}
 dyn_calf_front_laces_r {parent=base_calf_front_laces_r, loc=<0 0 0>, rot=<0 0 0 -1>}
 base_calf_back_laces_r {parent=calf_r, loc=<26.4551 11.3037 0.13>, rot=<-0.0009 0.0053 0.0317 -0.9995>}
 dyn_calf_back_laces_r {parent=base_calf_back_laces_r, loc=<0 0 0>, rot=<0 0 0 -1>}
 thigh_twist_01_r {parent=thigh_r, loc=<27.952 0 0>, rot=<0.0013 0 0 -1>}
 deform_groin_base_r {parent=thigh_r, loc=<-0.4801 -5.6776 0.2486>, rot=<0 0 0 -1>}
 deform_groin_r {parent=deform_groin_base_r, loc=<0 0 0>, rot=<0 0 0 1>}
 base_leg_laces_r {parent=thigh_r, loc=<16.581 -8.332 6.693>, rot=<0.9654 0.0068 -0.0349 0.2584>}
 dyn_leg_laces_r {parent=base_leg_laces_r, loc=<0 0 0>, rot=<0 0 0 1>}
 dyn_leg_laces_1_r {parent=dyn_leg_laces_r, loc=<0.229 0.3799 1.141>, rot=<-0.6118 0.2641 -0.3369 -0.6652>}
 dyn_leg_laces_0_r {parent=dyn_leg_laces_r, loc=<0.299 -0.3731 -1.1019>, rot=<0.6579 -0.3761 -0.3322 -0.5615>}
 dyn_leg_laces_2_r {parent=dyn_leg_laces_r, loc=<1.424 0.3822 -1.0101>, rot=<-0.0006 -0.0396 0.0351 -0.9986>}
 dyn_leg_laces_3_r {parent=dyn_leg_laces_2_r, loc=<3.9999 0 0>, rot=<0.0004 0.0266 -0.0146 -0.9995>}
 dyn_leg_laces_4_r {parent=dyn_leg_laces_3_r, loc=<4.2785 0 0>, rot=<0 0 0 1>}
 dyn_leg_laces_5_r {parent=dyn_leg_laces_4_r, loc=<4.3664 0 0>, rot=<0 0 0 1>}
 dyn_leg_laces_6_r {parent=dyn_leg_laces_r, loc=<1.4443 0.4006 0.6461>, rot=<0.9998 0.0214 0 -0.0004>}
 dyn_leg_laces_7_r {parent=dyn_leg_laces_6_r, loc=<2.6439 -0.005 0>, rot=<0 0 -0.0223 -0.9998>}
 dyn_leg_laces_8_r {parent=dyn_leg_laces_7_r, loc=<2.6227 0 0>, rot=<0 0.0104 0 -0.9999>}
 dyn_leg_laces_9_r {parent=dyn_leg_laces_8_r, loc=<3.0037 0 0>, rot=<0 0 0 1>}
 deform_glute_base_l {parent=pelvis, loc=<-7.3466 -9.0213 -9.1245>, rot=<0.0172 0.031 -0.4477 -0.8935>}
 deform_glute_l {parent=deform_glute_base_l, loc=<0 0 0>, rot=<0 0 0 1>}
 deform_glute_base_r {parent=pelvis, loc=<-7.3464 -9.0213 9.1245>, rot=<0.031 -0.0172 0.8935 -0.4477>}
 deform_glute_r {parent=deform_glute_base_r, loc=<0 0 0>, rot=<0 0 0 1>}
 base_heart_jnt {parent=pelvis, loc=<2.165 10.113 -0.009>, rot=<0 0 0.999 -0.0436>}
 dyn_heart_jnt {parent=base_heart_jnt, loc=<0 0 0>, rot=<0 0 0 1>}
 dyn_hip_laces_0_l {parent=dyn_heart_jnt, loc=<0.459 -0.3371 -3.529>, rot=<-0.0012 -0.0174 0.0697 -0.9974>}
 dyn_hip_laces_1_l {parent=dyn_hip_laces_0_l, loc=<3.1813 0 0>, rot=<0 0 0 1>}
 dyn_hip_laces_0_r {parent=dyn_heart_jnt, loc=<0.073 -0.264 3.447>, rot=<0.9972 -0.0741 -0.0087 -0.0006>}
 dyn_hip_laces_1_r {parent=dyn_hip_laces_0_r, loc=<4.0547 0 0>, rot=<0 0 0 1>}
 thigh_l_ctrl {controlType=EULER_TRANSFORM, type=Control, loc=<0 0 0>, rot=<0 0 0 1>}
 thigh_r_ctrl {controlType=EULER_TRANSFORM, type=Control, loc=<0 0 0>, rot=<0 0 0 1>}
 clavicle_l_ctrl {controlType=EULER_TRANSFORM, type=Control, loc=<0 0 0>, rot=<0 0 0 1>}
 upperarm_l_ctrl {controlType=EULER_TRANSFORM, parent=clavicle_l_ctrl, type=Control, loc=<0 0 0>, rot=<0 0 0.2176 0.976>}
 clavicle_r_ctrl {controlType=EULER_TRANSFORM, type=Control, loc=<0 0 0>, rot=<0 0 0 1>}
 upperarm_r_ctrl {controlType=EULER_TRANSFORM, parent=clavicle_r_ctrl, type=Control, loc=<0 0 0>, rot=<0 0 0 1>}
 head_ctrl {controlType=EULER_TRANSFORM, type=Control, loc=<0 0 0>, rot=<0 0 0 1>}
]
variables [
 IsSliding {type=bool, default=False}
 lowToGround {type=double, default=0}
]
