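# Physics asset for the left and right thigh laces on the BerryTartBrunt medium female body, apparently for secondary lace motion.
# Each side hangs a root lace bone off a base bone, branching into two short pieces and two four-bone chains, with narrow swing and zero twist limits.
# Base bones join the thighs with free angular motion, while larger thigh and thigh-twist capsules give the laces leg volumes.
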PhysicsAsset /BRCosmetics/Characters/Player/Female/Medium/Bodies/F_MED_BerryTartBrunt/Meshes/F_MED_BerryTartBrunt_Thigh_Laces_Physics
bodies {
 base_leg_laces_l [{shapeType=Capsule, length=1, radius=1, shapeName=base_leg_laces_l_capsule}]
 base_leg_laces_r [{shapeType=Capsule, length=1, radius=1, shapeName=base_leg_laces_r_capsule}]
 dyn_leg_laces_0_l [{
   shapeType=Capsule
   length=2.6161
   radius=1.0057
   shapeName=dyn_leg_laces_0_l_capsule
   center=<2.6691 0 0.5605>
   rotation=<rot 87.75 0.0055 0.0057>
  }]
 dyn_leg_laces_0_r [{
   shapeType=Capsule
   length=2.6161
   radius=1.0057
   shapeName=dyn_leg_laces_0_r_capsule
   center=<2.6691 0 0.5605>
   rotation=<rot 87.75 0.0055 0.0057>
  }]
 dyn_leg_laces_1_l [{shapeType=Capsule, length=2.1287, radius=0.8713, shapeName=dyn_leg_laces_1_l_capsule, center=<2.1706 0 -0.276>, rotation=<rot 90 0 0>}]
 dyn_leg_laces_1_r [{shapeType=Capsule, length=2.1287, radius=0.8713, shapeName=dyn_leg_laces_1_r_capsule, center=<2.1706 0 -0.276>, rotation=<rot 90 0 0>}]
 dyn_leg_laces_2_l [{shapeType=Capsule, length=2.1599, radius=0.8401, shapeName=dyn_leg_laces_2_l_capsule, center=<2 0 0>, rotation=<rot 90 0 0>}]
 dyn_leg_laces_2_r [{shapeType=Capsule, length=2.1919, radius=0.8081, shapeName=dyn_leg_laces_2_r_capsule, center=<2 0 0>, rotation=<rot 90 0 0>}]
 dyn_leg_laces_3_l [{shapeType=Capsule, length=2.0035, radius=0.9965, shapeName=dyn_leg_laces_3_l_capsule, center=<2 0 0>, rotation=<rot 90 0 0>}]
 dyn_leg_laces_3_r [{shapeType=Capsule, length=2.0541, radius=0.9459, shapeName=dyn_leg_laces_3_r_capsule, center=<2 0 0>, rotation=<rot 90 0 0>}]
 dyn_leg_laces_4_l [{shapeType=Capsule, length=1.9762, radius=1.0238, shapeName=dyn_leg_laces_4_l_capsule, center=<2 0 0>, rotation=<rot 90 0 0>}]
 dyn_leg_laces_4_r [{shapeType=Capsule, length=2.1043, radius=0.9974, shapeName=dyn_leg_laces_4_r_capsule, center=<2 0 0>, rotation=<rot 90 0 0>}]
 dyn_leg_laces_5_l [{shapeType=Capsule, length=2.0541, radius=0.9459, shapeName=dyn_leg_laces_5_l_capsule, center=<2 0 0>, rotation=<rot 90 0 0>}]
 dyn_leg_laces_5_r [{shapeType=Capsule, length=2.0843, radius=1.0174, shapeName=dyn_leg_laces_5_r_capsule, center=<2 -0.0349 0>, rotation=<rot 90 0 0>}]
 dyn_leg_laces_6_l [{shapeType=Capsule, length=2.1599, radius=0.8401, shapeName=dyn_leg_laces_6_l_capsule, center=<2 0 0>, rotation=<rot 90 0 0>}]
 dyn_leg_laces_6_r [{shapeType=Capsule, length=2.1919, radius=0.8081, shapeName=dyn_leg_laces_6_r_capsule, center=<2 0 0>, rotation=<rot 90 0 0>}]
 dyn_leg_laces_7_l [{shapeType=Capsule, length=2.0035, radius=0.9965, shapeName=dyn_leg_laces_7_l_capsule, center=<3.3064 0 0>, rotation=<rot 90 0 0>}]
 dyn_leg_laces_7_r [{shapeType=Capsule, length=2.0541, radius=0.9459, shapeName=dyn_leg_laces_7_r_capsule, center=<3.3064 0 0>, rotation=<rot 90 0 0>}]
 dyn_leg_laces_8_l [{shapeType=Capsule, length=1.9762, radius=1.0238, shapeName=dyn_leg_laces_8_l_capsule, center=<4.9224 0 0>, rotation=<rot 90 0 0>}]
 dyn_leg_laces_8_r [{
   shapeType=Capsule
   length=2.0941
   radius=1.0077
   shapeName=dyn_leg_laces_8_r_capsule
   center=<4.9224 0.2525 0>
   rotation=<rot 90 90.6248 88.0228>
  }]
 dyn_leg_laces_9_l [{shapeType=Capsule, length=2.0541, radius=0.9459, shapeName=dyn_leg_laces_9_l_capsule, center=<6.039 0 0>, rotation=<rot 90 0 0>}]
 dyn_leg_laces_9_r [{shapeType=Capsule, length=2.1242, radius=0.9775, shapeName=dyn_leg_laces_9_r_capsule, center=<6.039 0.4975 0>, rotation=<rot 90 0 0>}]
 dyn_leg_laces_l [{shapeType=Capsule, length=2, radius=1, shapeName=dyn_leg_laces_l_capsule, center=<2 -0.6427 0>, rotation=<rot 90 -180 -176.8465>}]
 dyn_leg_laces_r [{shapeType=Capsule, length=2, radius=1, shapeName=dyn_leg_laces_r_capsule, center=<2 0.6427 0>, rotation=<rot 90 -180 -176.8465>}]
 thigh_l [{shapeType=Capsule, length=5, radius=5, shapeName=thigh_l_capsule, rotation=<rot 90 0 0>}]
 thigh_r [{shapeType=Capsule, length=5, radius=5, shapeName=thigh_r_capsule, rotation=<rot 90 0 0>}]
 thigh_twist_01_l [{
   shapeType=Capsule
   length=20.7389
   radius=9.2612
   shapeName=thigh_twist_01_r_capsule
   center=<-1.5718 -1.0347 1.3912>
   rotation=<rot 81.7825 14.7899 5.8497>
  }]
 thigh_twist_01_r [{shapeType=Capsule, length=15.0685, radius=8.9728, shapeName=thigh_twist_01_r_capsule, center=<1.4031 1.4036 0>, rotation=<rot 90 0 -13.906>}]
}
constraints [
 {bone1Name=dyn_leg_laces_0_r, bone2Name=dyn_leg_laces_r, swing1LimitDegrees=10, swing2LimitDegrees=10, twistLimitDegrees=0}
 {bone1Name=dyn_leg_laces_1_r, bone2Name=dyn_leg_laces_r, swing1LimitDegrees=10, swing2LimitDegrees=10, twistLimitDegrees=0}
 {bone1Name=dyn_leg_laces_2_r, bone2Name=dyn_leg_laces_r, swing1LimitDegrees=40, swing2LimitDegrees=10, twistLimitDegrees=0}
 {bone1Name=dyn_leg_laces_3_r, bone2Name=dyn_leg_laces_2_r, swing1LimitDegrees=20, swing2LimitDegrees=10, twistLimitDegrees=0}
 {bone1Name=dyn_leg_laces_4_r, bone2Name=dyn_leg_laces_3_r, swing1LimitDegrees=8, swing2LimitDegrees=10, twistLimitDegrees=0}
 {bone1Name=dyn_leg_laces_5_r, bone2Name=dyn_leg_laces_4_r, swing1LimitDegrees=2, swing2LimitDegrees=10, twistLimitDegrees=0}
 {bone1Name=dyn_leg_laces_6_r, bone2Name=dyn_leg_laces_r, swing1LimitDegrees=20, swing2LimitDegrees=15, twistLimitDegrees=0}
 {bone1Name=dyn_leg_laces_7_r, bone2Name=dyn_leg_laces_6_r, swing1LimitDegrees=20, swing2LimitDegrees=15, twistLimitDegrees=0}
 {bone1Name=dyn_leg_laces_8_r, bone2Name=dyn_leg_laces_7_r, swing1LimitDegrees=20, swing2LimitDegrees=15, twistLimitDegrees=0}
 {bone1Name=dyn_leg_laces_9_r, bone2Name=dyn_leg_laces_8_r, swing1LimitDegrees=20, swing2LimitDegrees=15, twistLimitDegrees=0}
 {bone1Name=dyn_leg_laces_r, bone2Name=base_leg_laces_r, swing1LimitDegrees=5, swing2LimitDegrees=5, twistLimitDegrees=0}
 {bone1Name=dyn_leg_laces_l, bone2Name=base_leg_laces_l, swing1LimitDegrees=5, swing2LimitDegrees=5, twistLimitDegrees=0}
 {bone1Name=dyn_leg_laces_0_l, bone2Name=dyn_leg_laces_l, swing1LimitDegrees=10, swing2LimitDegrees=10, twistLimitDegrees=0}
 {bone1Name=dyn_leg_laces_1_l, bone2Name=dyn_leg_laces_l, swing1LimitDegrees=10, swing2LimitDegrees=10, twistLimitDegrees=0}
 {bone1Name=dyn_leg_laces_2_l, bone2Name=dyn_leg_laces_l, swing1LimitDegrees=40, swing2LimitDegrees=10, twistLimitDegrees=0}
 {bone1Name=dyn_leg_laces_3_l, bone2Name=dyn_leg_laces_2_l, swing1LimitDegrees=20, swing2LimitDegrees=10, twistLimitDegrees=0}
 {bone1Name=dyn_leg_laces_4_l, bone2Name=dyn_leg_laces_3_l, swing1LimitDegrees=10, swing2LimitDegrees=10, twistLimitDegrees=0}
 {bone1Name=dyn_leg_laces_5_l, bone2Name=dyn_leg_laces_4_l, swing1LimitDegrees=10, swing2LimitDegrees=10, twistLimitDegrees=0}
 {bone1Name=dyn_leg_laces_6_l, bone2Name=dyn_leg_laces_l, swing1LimitDegrees=20, swing2LimitDegrees=15, twistLimitDegrees=0}
 {bone1Name=dyn_leg_laces_7_l, bone2Name=dyn_leg_laces_6_l, swing1LimitDegrees=20, swing2LimitDegrees=15, twistLimitDegrees=0}
 {bone1Name=dyn_leg_laces_8_l, bone2Name=dyn_leg_laces_7_l, swing1LimitDegrees=20, swing2LimitDegrees=15, twistLimitDegrees=0}
 {bone1Name=dyn_leg_laces_9_l, bone2Name=dyn_leg_laces_8_l, swing1LimitDegrees=20, swing2LimitDegrees=15, twistLimitDegrees=0}
 {
  bone1Name=base_leg_laces_l
  bone2Name=thigh_l
  swing1LimitDegrees=45
  swing1Motion=Free
  swing2LimitDegrees=45
  swing2Motion=Free
  twistLimitDegrees=45
  twistMotion=Free
 }
 {
  bone1Name=base_leg_laces_r
  bone2Name=thigh_r
  swing1LimitDegrees=45
  swing1Motion=Free
  swing2LimitDegrees=45
  swing2Motion=Free
  twistLimitDegrees=45
  twistMotion=Free
 }
]
