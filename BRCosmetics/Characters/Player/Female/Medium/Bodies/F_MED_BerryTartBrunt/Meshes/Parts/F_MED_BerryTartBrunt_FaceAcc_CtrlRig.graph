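# Control Rig for the BerryTartBrunt face accessory mesh that makes two six-bone hair strands follow the body.
# Control offset nulls are parent-constrained to head, neck, spine and clavicles, then feed twist and Hermite spline IK functions per strand.
# Arm-driven correctives follow: right upperarm swing weights IK bone offsets, while a spherical pose reader nudges the left strand.
# A debug branch, off by default, drives body bones from test controls.
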ControlRigBlueprint /BRCosmetics/Characters/Player/Female/Medium/Bodies/F_MED_BerryTartBrunt/Meshes/Parts/F_MED_BerryTartBrunt_FaceAcc_CtrlRig
graphs [
 New Function_ContainedGraph {nodes [
   n1 = RigVMFunctionEntryNode()
   RigVMFunction_ControlFlowBranch = &RigVMFunction_ControlFlowBranch(ExecuteContext=@n1.ExecuteContext, Condition=@n1.debug)
   RigVMFunctionReturnNode(ExecuteContext=@RigVMFunction_ControlFlowBranch.Completed)
   RigUnit_ItemArray = &RigUnit_ItemArray(Items="((Type=Bone,Name=\"spine_01\"),(Type=Bone,Name=\"spine_02\"),(Type=Bone,Name=\"spine_03\"),(Type=Bone,Name=\"spine_04\"),(Type=Bone,Name=\"spine_05\"),(Type=Bone,Name=\"neck_01\"),(Type=Bone,Name=\"neck_02\"),(Type=Bone,Name=\"head\"),(Type=Bone,Name=\"clavicle_l\"),(Type=B ...(truncated)")
   DISPATCH_RigVMDispatch_ArrayIterator = RigVMDispatchNode(Array=@RigUnit_ItemArray.Items, ExecuteContext=@RigVMFunction_ControlFlowBranch.True) {template="DISPATCH_RigVMDispatch_ArrayIterator(in Array,out Element,out Index,out Count,out Ratio)"}
   Concat = /Script/RigVM.RigVMFunction_NameConcat(B=_ctrl, A=@DISPATCH_RigVMDispatch_ArrayIterator.Element.Name) {is=struct RigVMUnitNode, template="Concat::Execute(in A,in B,out Result)"}
   > &RigUnit_SetTransform(
    Item=(Type=Bone,Name="spine_01")
    Weight=1
    bInitial=False
    Item.Name=@DISPATCH_RigVMDispatch_ArrayIterator.Element.Name
    Value=&RigUnit_GetTransform(Item=(Type=Control,Name="spine_01_ctrl"), bInitial=False, Item.Name=@Concat.Result).Transform
   )
  ]}
 New Function_ContainedGraph {nodes [
   n1 = RigVMFunctionEntryNode()
   RigVMFunction_Sequence = &RigVMFunction_Sequence(ExecuteContext=@n1.ExecuteContext)
   TwistBones_2 = RigVMFunctionReferenceNode(
    Axis to Twist=(bX=true,bY=true,bZ=true)
    Bones to Twist=((Type=Bone,Name="dyn_hair_2_l"),(Type=Bone,Name="dyn_hair_3_l"),(Type=Bone,Name="dyn_hair_4_l"),(Type=Bone,Name="dyn_hair_5_l"))
    End=(Type=Bone,Name="neck_01")
    Start=(Type=Bone,Name="pelvis")
    ExecuteContext=@RigVMFunction_Sequence.A
   )
   > TwistBones_2_1 = RigVMFunctionReferenceNode(
    Axis to Twist=(bX=true,bY=true,bZ=true)
    Bones to Twist=((Type=Bone,Name="dyn_hair_2_r"),(Type=Bone,Name="dyn_hair_3_r"),(Type=Bone,Name="dyn_hair_4_r"),(Type=Bone,Name="dyn_hair_5_r"))
    End=(Type=Bone,Name="neck_01")
    Start=(Type=Bone,Name="pelvis")
   )
   RigUnit_ItemArray = &RigUnit_ItemArray(Items=((Type=Bone,Name="dyn_hair_0_l"),(Type=Bone,Name="dyn_hair_1_l"),(Type=Bone,Name="dyn_hair_2_l"),(Type=Bone,Name="dyn_hair_3_l"),(Type=Bone,Name="dyn_hair_4_l"),(Type=Bone,Name="dyn_hair_5_l")))
   n2 = &RigUnit_ItemArray(Items=((Type=Control,Name="dyn_hair_0_l_ctrl"),(Type=Control,Name="dyn_hair_1_l_ctrl"),(Type=Control,Name="dyn_hair_2_l_ctrl"),(Type=Control,Name="dyn_hair_3_l_ctrl"),(Type=Control,Name="dyn_hair_4_l_ctrl"),(Type=Control,Name="dyn_hair_5_l_ctrl")))
   SplineIK = RigVMFunctionReferenceNode(
    Primary Axis=(X=1,Y=0,Z=0)
    Samples Per Segment=16
    Secondary Spline Direction=(X=0,Y=0,Z=0)
    Spline Mode=Hermite
    Stretch=true
    Up Axis=(X=0,Y=0,Z=0)
    debug=false
    Bones=@RigUnit_ItemArray.Items
    Controls=@n2.Items
    ExecuteContext=@RigVMFunction_Sequence.B
   )
   n3 = &RigUnit_ItemArray(Items=((Type=Control,Name="dyn_hair_0_r_ctrl"),(Type=Control,Name="dyn_hair_1_r_ctrl"),(Type=Control,Name="dyn_hair_2_r_ctrl"),(Type=Control,Name="dyn_hair_3_r_ctrl"),(Type=Control,Name="dyn_hair_4_r_ctrl"),(Type=Control,Name="dyn_hair_5_r_ctrl")))
   n4 = &RigUnit_ItemArray(Items=((Type=Bone,Name="dyn_hair_0_r"),(Type=Bone,Name="dyn_hair_1_r"),(Type=Bone,Name="dyn_hair_2_r"),(Type=Bone,Name="dyn_hair_3_r"),(Type=Bone,Name="dyn_hair_4_r"),(Type=Bone,Name="dyn_hair_5_r")))
   > SplineIK_1 = RigVMFunctionReferenceNode(
    Primary Axis=(X=1,Y=0,Z=0)
    Samples Per Segment=16
    Secondary Spline Direction=(X=0,Y=0,Z=0)
    Spline Mode=Hermite
    Stretch=true
    Up Axis=(X=0,Y=0,Z=0)
    Controls=@n3.Items
    Bones=@n4.Items
   )
   > RigVMFunctionReturnNode()
  ]}
 RigVMFunctionLibrary {nodes [
   controllers_FNC = RigVMCollapseNode(ExecuteContext=(), debug=False)
   spline_IK_FNC = RigVMCollapseNode(ExecuteContext=())
  ]}
 RigVMModel {nodes [
   RigUnit_BeginExecution = &RigUnit_BeginExecution()
   controllers_FNC = RigVMFunctionReferenceNode(debug=false, ExecuteContext=@RigUnit_BeginExecution.ExecutePin)
   > &RigUnit_ParentConstraint(
    AdvancedSettings=(InterpolationType=Average,RotationOrderForFilter=XZY)
    Child=(Type=Null,Name="dyn_hair_0_l_ctrl_offset")
    Parents=((Item=(Type=Bone,Name="head"),Weight=1))
    Weight=1
   )
   > &RigUnit_ParentConstraint(
    AdvancedSettings=(InterpolationType=Average,RotationOrderForFilter=XZY)
    Child=(Type=Null,Name="dyn_hair_0_r_ctrl_offset")
    Parents=((Item=(Type=Bone,Name="head"),Weight=1))
    Weight=1
   )
   > &RigUnit_ParentConstraint(
    AdvancedSettings=(InterpolationType=Average,RotationOrderForFilter=XZY)
    Child=(Type=Null,Name="dyn_hair_1_l_ctrl_offset")
    Parents=((Item=(Type=Bone,Name="head"),Weight=1))
    Weight=1
   )
   > &RigUnit_ParentConstraint(
    AdvancedSettings=(InterpolationType=Average,RotationOrderForFilter=XZY)
    Child=(Type=Null,Name="dyn_hair_1_r_ctrl_offset")
    Parents=((Item=(Type=Bone,Name="head"),Weight=1))
    Weight=1
   )
   > &RigUnit_ParentConstraint(
    AdvancedSettings=(InterpolationType=Average,RotationOrderForFilter=XZY)
    Child=(Type=Null,Name="dyn_hair_2_l_ctrl_offset")
    Parents=((Item=(Type=Bone,Name="neck_01"),Weight=1))
    Weight=1
   )
   > &RigUnit_ParentConstraint(
    AdvancedSettings=(InterpolationType=Average,RotationOrderForFilter=XZY)
    Child=(Type=Null,Name="dyn_hair_2_r_ctrl_offset")
    Parents=((Item=(Type=Bone,Name="neck_01"),Weight=1))
    Weight=1
   )
   > &RigUnit_ParentConstraint(
    AdvancedSettings=(InterpolationType=Average,RotationOrderForFilter=XZY)
    Child=(Type=Null,Name="dyn_hair_3_l_ctrl_offset")
    Parents=((Item=(Type=Bone,Name="spine_05"),Weight=1),(Item=(Type=Bone,Name="clavicle_l"),Weight=0))
    Weight=1
   )
   > &RigUnit_ParentConstraint(
    AdvancedSettings=(InterpolationType=Average,RotationOrderForFilter=XZY)
    Child=(Type=Null,Name="dyn_hair_3_r_ctrl_offset")
    Parents=((Item=(Type=Bone,Name="spine_05"),Weight=1),(Item=(Type=Bone,Name="clavicle_r"),Weight=0))
    Weight=1
   )
   > &RigUnit_ParentConstraint(
    AdvancedSettings=(InterpolationType=Average,RotationOrderForFilter=XZY)
    Child=(Type=Null,Name="dyn_hair_4_l_ctrl_offset")
    Parents=((Item=(Type=Bone,Name="spine_04"),Weight=1))
    Weight=1
   )
   > &RigUnit_ParentConstraint(
    AdvancedSettings=(InterpolationType=Average,RotationOrderForFilter=XZY)
    Child=(Type=Null,Name="dyn_hair_4_r_ctrl_offset")
    Parents=((Item=(Type=Bone,Name="spine_04"),Weight=1))
    Weight=1
   )
   > &RigUnit_ParentConstraint(
    AdvancedSettings=(InterpolationType=Average,RotationOrderForFilter=XZY)
    Child=(Type=Null,Name="dyn_hair_5_l_ctrl_offset")
    Parents=((Item=(Type=Bone,Name="spine_03"),Weight=1),(Item=(Type=Bone,Name="spine_02"),Weight=1))
    Weight=1
   )
   > &RigUnit_ParentConstraint(
    AdvancedSettings=(InterpolationType=Average,RotationOrderForFilter=XZY)
    Child=(Type=Null,Name="dyn_hair_5_r_ctrl_offset")
    Parents=((Item=(Type=Bone,Name="spine_03"),Weight=1),(Item=(Type=Bone,Name="spine_02"),Weight=1))
    Weight=1
   )
   > &RigUnit_ParentConstraint(
    AdvancedSettings=(InterpolationType=Average,RotationOrderForFilter=XZY)
    Child=(Type=Null,Name="dyn_hair_3_l_ctrl_offset")
    Parents=((Item=(Type=Bone,Name="clavicle_l"),Weight=1))
    Weight=0.8
   )
   > &RigUnit_ParentConstraint(
    AdvancedSettings=(InterpolationType=Average,RotationOrderForFilter=XZY)
    Child=(Type=Null,Name="dyn_hair_4_l_ctrl_offset")
    Parents=((Item=(Type=Bone,Name="clavicle_l"),Weight=1))
    Weight=0.2
   )
   > &RigUnit_ParentConstraint(
    AdvancedSettings=(InterpolationType=Average,RotationOrderForFilter=XZY)
    Child=(Type=Null,Name="dyn_hair_3_r_ctrl_offset")
    Parents=((Item=(Type=Bone,Name="clavicle_r"),Weight=1))
    Weight=0.8
   )
   > &RigUnit_ParentConstraint(
    AdvancedSettings=(InterpolationType=Average,RotationOrderForFilter=XZY)
    Child=(Type=Null,Name="dyn_hair_4_r_ctrl_offset")
    Parents=((Item=(Type=Bone,Name="clavicle_r"),Weight=1))
    Weight=0.2
   )
   > spline_IK_FNC = RigVMFunctionReferenceNode()
   > &RigUnit_ParentConstraint(
    AdvancedSettings=(InterpolationType=Average,RotationOrderForFilter=XZY)
    Child=(Type=Bone,Name="upperarm_r_2_2")
    Parents=((Item=(Type=Bone,Name="upperarm_r"),Weight=1))
    Weight=1
   )
   n1 = &RigVMFunction_MathQuaternionSwingTwist(
    TwistAxis=(X=1,Y=0,Z=0)
    Input=&RigUnit_GetTransform(Item=(Type=Bone,Name="upperarm_r_2_2"), Space=LocalSpace, bInitial=false).Transform.Rotation
   )
   Remap = &RigVMFunction_MathDoubleRemap(
    SourceMaximum=60
    SourceMinimum=0
    TargetMaximum=1
    TargetMinimum=0
    Value=&RigVMFunction_MathDoubleClamp(Maximum=60, Value=&RigVMFunction_MathQuaternionToEuler(RotationOrder=ZYX, Value=@n1.Swing).Result.Z).Result
   )
   > &RigUnit_ModifyTransforms(
    ItemToModify="((Item=(Type=Bone,Name=\"dyn_hair_3_ik_r\"),Transform=(Rotation=(X=0,Y=0,Z=0,W=1),Translation=(X=0,Y=2.2,Z=2),Scale3D=(X=1,Y=1,Z=0.01))),(Item=(Type=Bone,Name=\"dyn_hair_2_ik_r\"),Transform=(Rotation=(X=0,Y=0,Z=0,W=1),Translation=(X=0,Y=0,Z=4),Scale3D=(X=1,Y=1 ...(truncated)"
    Mode=OverrideLocal
    WeightMinimum=0
    Weight=@Remap.Result
   )
   > n2 = &RigUnit_SphericalPoseReader(
    ActiveRegionScaleFactors=(PositiveWidth=2,NegativeWidth=3,PositiveHeight=1,NegativeHeight=1)
    ActiveRegionSize=0.1
    Debug=(bDrawDebug=false,bDraw2D=False,bDrawLocalAxes=False,DebugScale=25,DebugSegments=20,DebugThickness=0.25)
    DriverAxis=(X=-1,Y=0,Z=0)
    DriverItem=(Type=Bone,Name="upperarm_l")
    FalloffRegionScaleFactors=(PositiveWidth=4,NegativeWidth=-10,PositiveHeight=2,NegativeHeight=2)
    FalloffSize=0.28
    FlipHeightScaling=False
    FlipWidthScaling=False
    OptionalParentItem=(Type=Bone,Name="spine_05")
    RotationOffset=(X=-90,Y=-40,Z=15)
   )
   > &RigUnit_ModifyTransforms(
    ItemToModify=((Item=(Type=Bone,Name="dyn_hair_3_l"),Transform=(Rotation=(X=0,Y=0,Z=0,W=1),Translation=(X=0,Y=0,Z=1),Scale3D=(X=1,Y=1,Z=1))))
    WeightMinimum=0
    Weight=@n2.OutputParam
   )
  ]}
]
hierarchy [
 root {loc=<0 0 0>, rot=<0 0 0 1>}
 pelvis {parent=root, loc=<0 -0.5301 98.0571>, rot=<0.0127 0.707 -0.0127 -0.707>}
 spine_01 {parent=pelvis, loc=<6.1208 -0.9328 0>, rot=<0 0 0.034 -0.9994>}
 spine_02 {parent=spine_01, loc=<6.5997 -0.2095 0>, rot=<0 0 0.0238 -0.9997>}
 spine_03 {parent=spine_02, loc=<6.4914 -0.1237 0>, rot=<0 0 0.0149 -0.9999>}
 spine_04 {parent=spine_03, loc=<6.9598 0.0794 0>, rot=<0 0 -0.0132 -0.9999>}
 spine_05 {parent=spine_04, loc=<6.9987 0 0>, rot=<0 0 -0.0015 -1>}
 neck_01 {parent=spine_05, loc=<11.3138 0.1529 0>, rot=<0 0 -0.2003 -0.9797>}
 neck_02 {parent=neck_01, loc=<4.5627 0 0>, rot=<0 0 0.0504 -0.9987>}
 head {parent=neck_02, loc=<5.3914 0 0>, rot=<0 0 0.115 -0.9934>}
 base_hair_0_l {parent=head, loc=<4.2289 4.7218 -7.0715>, rot=<0.0016 0.0325 0.998 -0.0546>}
 dyn_hair_0_l {parent=base_hair_0_l, loc=<0 0 0>, rot=<0 0 0 1>}
 dyn_hair_1_l {parent=dyn_hair_0_l, loc=<7.8587 0 0>, rot=<-0.003 -0.0534 0.0552 -0.997>}
 dyn_hair_2_l {parent=dyn_hair_1_l, loc=<7.8165 0 0>, rot=<-0.0154 -0.1226 0.0741 -0.9896>}
 dyn_hair_3_l {parent=dyn_hair_2_l, loc=<7.9969 0 0>, rot=<-0.0011 0.0924 -0.0119 -0.9956>}
 dyn_hair_4_l {parent=dyn_hair_3_l, loc=<7.8165 0 0>, rot=<-0.0234 0.1259 -0.0928 -0.9874>}
 dyn_hair_5_l {parent=dyn_hair_4_l, loc=<7.8164 0 0>, rot=<0 0 0 1>}
 dyn_hair_5_ik_l {parent=dyn_hair_5_l, loc=<0 0 0>, rot=<0 0 0 1>}
 dyn_hair_4_ik_l {parent=dyn_hair_4_l, loc=<0 0 0>, rot=<0 0 0 1>}
 dyn_hair_3_ik_l {parent=dyn_hair_3_l, loc=<0 0 0>, rot=<0 0 0 1>}
 dyn_hair_2_ik_l {parent=dyn_hair_2_l, loc=<0 0 0>, rot=<0 0 0 1>}
 dyn_hair_1_ik_l {parent=dyn_hair_1_l, loc=<0 0 0>, rot=<0 0 0 1>}
 dyn_hair_0_ik_l {parent=dyn_hair_0_l, loc=<0 0 0>, rot=<0 0 0 1>}
 base_hair_0_r {parent=head, loc=<4.2289 4.7218 7.0715>, rot=<0.0546 -0.998 -0.0325 -0.0016>}
 dyn_hair_0_r {parent=base_hair_0_r, loc=<0 0 0>, rot=<0 0 0 1>}
 dyn_hair_1_r {parent=dyn_hair_0_r, loc=<7.859 0 0>, rot=<0.003 -0.0534 -0.0552 -0.997>}
 dyn_hair_2_r {parent=dyn_hair_1_r, loc=<7.8159 0 -0.0001>, rot=<0.0154 -0.1226 -0.0741 -0.9896>}
 dyn_hair_3_r {parent=dyn_hair_2_r, loc=<7.9968 0 -0.0001>, rot=<0.0011 0.0924 0.0119 -0.9956>}
 dyn_hair_4_r {parent=dyn_hair_3_r, loc=<7.8169 -0.0001 0.0001>, rot=<0.0234 0.1259 0.0928 -0.9874>}
 dyn_hair_5_r {parent=dyn_hair_4_r, loc=<7.8169 0 0>, rot=<0 0 0 1>}
 dyn_hair_5_ik_r {parent=dyn_hair_5_r, loc=<0 0 0>, rot=<0 0 0 1>}
 dyn_hair_4_ik_r {parent=dyn_hair_4_r, loc=<0 0 0>, rot=<0 0 0 1>}
 dyn_hair_3_ik_r {parent=dyn_hair_3_r, loc=<0 0 0>, rot=<0 0 0 1>}
 dyn_hair_2_ik_r {parent=dyn_hair_2_r, loc=<0 0 0>, rot=<0 0 0 1>}
 dyn_hair_1_ik_r {parent=dyn_hair_1_r, loc=<0 0 0>, rot=<0 0 0 1>}
 dyn_hair_0_ik_r {parent=dyn_hair_0_r, loc=<0 0 0>, rot=<0 0 0 1>}
 base_crown {parent=head, loc=<10.2638 1.4068 0>, rot=<0.4979 0.5021 0.5021 0.4979>}
 dyn_crown {parent=base_crown, loc=<0 0 0>, rot=<0 0 0 1>}
 base_frontHair {parent=head, loc=<10.2119 7.6884 0>, rot=<0.4979 0.5021 0.5021 0.4979>}
 dyn_frontHair {parent=base_frontHair, loc=<0 0 0>, rot=<0 0 0 1>}
 clavicle_l {parent=spine_05, loc=<5.4831 -0.1053 -3.2946>, rot=<0.0465 -0.7289 -0.009 -0.6829>}
 upperarm_l {parent=clavicle_l, loc=<9.8658 -1.2136 -0.4105>, rot=<0.0164 -0.3245 -0.0269 -0.9454>}
 lowerarm_l {parent=upperarm_l, loc=<26.465 -0.8345 0.259>, rot=<0.0036 0.0161 -0.2298 -0.9731>}
 hand_l {parent=lowerarm_l, loc=<22.7355 -0.3203 -0.6105>, rot=<-0.6784 -0.0223 0.0735 -0.7307>}
 thumb_01_l {parent=hand_l, loc=<1.3118 -0.3191 -2.459>, rot=<0.6546 -0.4042 0.0664 -0.6354>}
 thumb_02_l {parent=thumb_01_l, loc=<5.9336 -0.0448 0.0401>, rot=<-0.0037 0.0213 0.298 -0.9543>}
 thumb_03_l {parent=thumb_02_l, loc=<2.2223 0.3682 -0.0907>, rot=<0.0118 -0.013 0.0066 -0.9998>}
 index_metacarpal_l {parent=hand_l, loc=<1.6535 0.0183 -2.96>, rot=<0.0158 -0.0124 -0.0173 -0.9996>}
 index_01_l {parent=index_metacarpal_l, loc=<8.3008 0.0357 -0.0524>, rot=<-0.0044 0.0485 0.0924 -0.9945>}
 index_02_l {parent=index_01_l, loc=<3.3587 -0.3004 0.0452>, rot=<-0.002 0.0027 0.3225 -0.9466>}
 index_03_l {parent=index_02_l, loc=<2.5578 -0.123 -0.0252>, rot=<-0.002 0.0067 0.2558 -0.9667>}
 middle_metacarpal_l {parent=hand_l, loc=<2.1856 -0.1725 -0.5984>, rot=<0.006 0.0168 -0.0329 -0.9993>}
 middle_01_l {parent=middle_metacarpal_l, loc=<7.3275 -0.2447 0.0129>, rot=<-0.0208 0.0042 0.0858 -0.9961>}
 middle_02_l {parent=middle_01_l, loc=<4.0123 0.1401 0.1175>, rot=<-0.0218 0.0258 0.3512 -0.9357>}
 middle_03_l {parent=middle_02_l, loc=<2.6077 0.0534 0.1373>, rot=<0.0082 0.0188 0.1708 -0.9851>}
 ring_metacarpal_l {parent=hand_l, loc=<2.2074 -0.3625 1.3638>, rot=<-0.0284 0.0722 -0.0458 -0.9959>}
 ring_01_l {parent=ring_metacarpal_l, loc=<7.0295 -0.6197 -0.1446>, rot=<-0.0489 -0.0252 0.1892 -0.9804>}
 ring_02_l {parent=ring_01_l, loc=<3.6174 0.4647 -0.1027>, rot=<0 0 0.2806 -0.9598>}
 ring_03_l {parent=ring_02_l, loc=<2.4152 0.114 0.0199>, rot=<0 0 0.1189 -0.9929>}
 pinky_metacarpal_l {parent=hand_l, loc=<1.7616 -0.9753 3.3814>, rot=<-0.0397 0.0814 -0.069 -0.9935>}
 pinky_01_l {parent=pinky_metacarpal_l, loc=<7.1709 -0.4831 -0.1094>, rot=<-0.0654 -0.0035 0.2242 -0.9723>}
 pinky_02_l {parent=pinky_01_l, loc=<2.8274 0.0893 -0.1043>, rot=<0.0037 0.0011 0.2854 -0.9584>}
 pinky_03_l {parent=pinky_02_l, loc=<2.2046 0.0307 -0.0078>, rot=<-0.0002 0.0098 0.1445 -0.9895>}
 deform_wrist_dn_base_l {parent=hand_l, loc=<0 -1.5326 0>, rot=<0 0 0 -1>}
 deform_wrist_dn_l {parent=deform_wrist_dn_base_l, loc=<0 0 0>, rot=<0 0 0 1>}
 deform_wrist_up_base_l {parent=hand_l, loc=<0 2.3009 0>, rot=<0 0 0 -1>}
 deform_wrist_up_l {parent=deform_wrist_up_base_l, loc=<0 0 0>, rot=<0 0 0 1>}
 lowerarm_twist_02_l {parent=lowerarm_l, loc=<10.6707 0 0>, rot=<0 0 0 1>}
 lowerarm_twist_01_l {parent=lowerarm_l, loc=<18.316 0 0>, rot=<0 0 0 1>}
 deform_elbow_in_base_l {parent=lowerarm_l, loc=<0.1834 3.6062 -0.2572>, rot=<0 0 0 -1>}
 deform_elbow_in_l {parent=deform_elbow_in_base_l, loc=<0 0 0>, rot=<0 0 0 1>}
 deform_elbow_out_base_l {parent=lowerarm_l, loc=<0 -4 0>, rot=<0 0 0 1>}
 deform_elbow_out_l {parent=deform_elbow_out_base_l, loc=<0 0 0>, rot=<0 0 0 1>}
 upperarm_twist_01_l {parent=upperarm_l, loc=<7.2135 -0.1134 0>, rot=<0 0 0 1>}
 deform_upperarm_bk_base_l {parent=upperarm_twist_01_l, loc=<-6.4053 -4.3746 0.314>, rot=<0 0 0 -1>}
 deform_upperarm_bk_l {parent=deform_upperarm_bk_base_l, loc=<0 0 0>, rot=<0 0 0 1>}
 deform_upperarm_fr_base_l {parent=upperarm_twist_01_l, loc=<-5.6312 5.092 0.3354>, rot=<0 0 0 -1>}
 deform_upperarm_fr_l {parent=deform_upperarm_fr_base_l, loc=<0 0 0>, rot=<0 0 0 1>}
 deform_upperarm_up_base_l {parent=upperarm_twist_01_l, loc=<-8.891 0.4825 3.9517>, rot=<0 0 0 -1>}
 deform_upperarm_up_l {parent=deform_upperarm_up_base_l, loc=<0 0 0>, rot=<0 0 0 1>}
 upperarm_twist_02_l {parent=upperarm_l, loc=<19.7998 -0.1134 0>, rot=<0 0 0 1>}
 clavicle_r {parent=spine_05, loc=<5.4831 -0.1053 3.2946>, rot=<0.7289 0.0465 -0.6829 0.009>}
 upperarm_r {parent=clavicle_r, loc=<-9.8658 1.2136 0.4105>, rot=<0.0164 -0.3245 -0.0269 -0.9454>}
 lowerarm_r {parent=upperarm_r, loc=<-26.465 0.8345 -0.259>, rot=<0.0036 0.0161 -0.2298 -0.9731>}
 hand_r {parent=lowerarm_r, loc=<-22.7355 0.3203 0.6106>, rot=<-0.6784 -0.0223 0.0735 -0.7307>}
 thumb_01_r {parent=hand_r, loc=<-1.3118 0.3191 2.459>, rot=<0.6546 -0.4042 0.0664 -0.6354>}
 thumb_02_r {parent=thumb_01_r, loc=<-5.9336 0.0448 -0.0401>, rot=<-0.0037 0.0213 0.298 -0.9543>}
 thumb_03_r {parent=thumb_02_r, loc=<-2.2223 -0.3682 0.0907>, rot=<0.0118 -0.013 0.0066 -0.9998>}
 index_metacarpal_r {parent=hand_r, loc=<-1.6535 -0.0183 2.96>, rot=<0.0158 -0.0124 -0.0173 -0.9996>}
 index_01_r {parent=index_metacarpal_r, loc=<-8.3008 -0.0357 0.0524>, rot=<-0.0044 0.0485 0.0924 -0.9945>}
 index_02_r {parent=index_01_r, loc=<-3.3587 0.3004 -0.0452>, rot=<-0.002 0.0027 0.3225 -0.9466>}
 index_03_r {parent=index_02_r, loc=<-2.5578 0.123 0.0252>, rot=<-0.002 0.0067 0.2558 -0.9667>}
 middle_metacarpal_r {parent=hand_r, loc=<-2.1856 0.1725 0.5984>, rot=<0.006 0.0168 -0.0329 -0.9993>}
 middle_01_r {parent=middle_metacarpal_r, loc=<-7.3275 0.2447 -0.0129>, rot=<-0.0208 0.0042 0.0858 -0.9961>}
 middle_02_r {parent=middle_01_r, loc=<-4.0123 -0.1401 -0.1175>, rot=<-0.0218 0.0258 0.3512 -0.9357>}
 middle_03_r {parent=middle_02_r, loc=<-2.6077 -0.0534 -0.1373>, rot=<0.0082 0.0188 0.1708 -0.9851>}
 ring_metacarpal_r {parent=hand_r, loc=<-2.2074 0.3625 -1.3638>, rot=<-0.0284 0.0722 -0.0458 -0.9959>}
 ring_01_r {parent=ring_metacarpal_r, loc=<-7.0295 0.6197 0.1446>, rot=<-0.0489 -0.0252 0.1892 -0.9804>}
 ring_02_r {parent=ring_01_r, loc=<-3.6174 -0.4647 0.1027>, rot=<0 0 0.2806 -0.9598>}
 ring_03_r {parent=ring_02_r, loc=<-2.4152 -0.114 -0.0199>, rot=<0 0 0.1189 -0.9929>}
 pinky_metacarpal_r {parent=hand_r, loc=<-1.7616 0.9753 -3.3814>, rot=<-0.0397 0.0814 -0.069 -0.9935>}
 pinky_01_r {parent=pinky_metacarpal_r, loc=<-7.1709 0.4831 0.1094>, rot=<-0.0654 -0.0035 0.2242 -0.9723>}
 pinky_02_r {parent=pinky_01_r, loc=<-2.8274 -0.0893 0.1043>, rot=<0.0037 0.0011 0.2854 -0.9584>}
 pinky_03_r {parent=pinky_02_r, loc=<-2.2046 -0.0307 0.0078>, rot=<-0.0002 0.0098 0.1445 -0.9895>}
 deform_wrist_dn_base_r {parent=hand_r, loc=<-0.0001 1.5328 0>, rot=<0 0 0 -1>}
 deform_wrist_dn_r {parent=deform_wrist_dn_base_r, loc=<0 0 0>, rot=<0 0 0 1>}
 deform_wrist_up_base_r {parent=hand_r, loc=<-0.0002 -2.3007 -0.0001>, rot=<0 0 0 -1>}
 deform_wrist_up_r {parent=deform_wrist_up_base_r, loc=<0 0 0>, rot=<0 0 0 1>}
 lowerarm_twist_02_r {parent=lowerarm_r, loc=<-10.6707 0 0>, rot=<0 0 0 -1>}
 lowerarm_twist_01_r {parent=lowerarm_r, loc=<-18.316 0 0>, rot=<0 0 0 -1>}
 deform_elbow_in_base_r {parent=lowerarm_r, loc=<-0.1834 -3.6062 0.2572>, rot=<0 0 0 -1>}
 deform_elbow_in_r {parent=deform_elbow_in_base_r, loc=<0 0 0>, rot=<0 0 0 1>}
 deform_elbow_out_base_r {parent=lowerarm_r, loc=<0 4 -0.0001>, rot=<0 0 0 -1>}
 deform_elbow_out_r {parent=deform_elbow_out_base_r, loc=<0 0 0>, rot=<0 0 0 1>}
 upperarm_twist_01_r {parent=upperarm_r, loc=<-7.2135 0.1134 0>, rot=<0 0 0 -1>}
 deform_upperarm_bk_base_r {parent=upperarm_twist_01_r, loc=<6.4053 4.3746 -0.314>, rot=<0 0 0 -1>}
 deform_upperarm_bk_r {parent=deform_upperarm_bk_base_r, loc=<0 0 0>, rot=<0 0 0 1>}
 deform_upperarm_fr_base_r {parent=upperarm_twist_01_r, loc=<5.6312 -5.092 -0.3354>, rot=<0 0 0 -1>}
 deform_upperarm_fr_r {parent=deform_upperarm_fr_base_r, loc=<0 0 0>, rot=<0 0 0 1>}
 deform_upperarm_up_base_r {parent=upperarm_twist_01_r, loc=<8.891 -0.4825 -3.9517>, rot=<0 0 0 -1>}
 deform_upperarm_up_r {parent=deform_upperarm_up_base_r, loc=<0 0 0>, rot=<0 0 0 1>}
 upperarm_twist_02_r {parent=upperarm_r, loc=<-19.7998 0.1134 0>, rot=<0 0 0 -1>}
 upperarm_r_2 {parent=clavicle_r, loc=<-9.8658 1.2136 0.4105>, rot=<0.0164 -0.3245 -0.0269 -0.9454>}
 upperarm_r_2_2 {parent=upperarm_r_2, loc=<0 0 0>, rot=<0 0 0 1>}
 deform_pec_aim_base_l {parent=spine_04, loc=<3.2456 9.3309 -1.2237>, rot=<0.0603 -0.4033 0.3318 -0.8507>}
 deform_pec_aim_l {parent=deform_pec_aim_base_l, loc=<0 0 0>, rot=<0 0 0 1>}
 deform_pec_base_l {parent=deform_pec_aim_l, loc=<9.3665 0 0>, rot=<0 0 0 1>}
 deform_pec_l {parent=deform_pec_base_l, loc=<0 0 0>, rot=<0 0 0 1>}
 deform_pec_aim_base_r {parent=spine_04, loc=<3.2452 9.3309 1.2237>, rot=<-0.4033 -0.0603 0.8507 0.3318>}
 deform_pec_aim_r {parent=deform_pec_aim_base_r, loc=<0 0 0>, rot=<0 0 0 1>}
 deform_pec_base_r {parent=deform_pec_aim_r, loc=<-9.3667 -0.0003 -0.0004>, rot=<0 0 0 1>}
 deform_pec_r {parent=deform_pec_base_r, loc=<0 0 0>, rot=<0 0 0 1>}
 thigh_l {parent=pelvis, loc=<-1.0855 -0.2109 -8.7053>, rot=<0.0009 0.0355 0.0155 -0.9993>}
 calf_l {parent=thigh_l, loc=<-42.5284 -0.0019 -0.7297>, rot=<0.0006 -0.0161 -0.0391 -0.9991>}
 foot_l {parent=calf_l, loc=<-46.7401 -2.5014 -0.592>, rot=<0.0427 -0.0238 0.0353 -0.9982>}
 ball_l {parent=foot_l, loc=<-3.5658 10.8066 -0.1544>, rot=<0.0001 -0.0001 -0.7063 -0.7079>}
 calf_twist_02_l {parent=calf_l, loc=<-19.1546 -1.7943 -0.2638>, rot=<0.0237 0.0095 -0.0035 -0.9997>}
 calf_twist_01_l {parent=calf_l, loc=<-35.8483 -1.9197 -0.5794>, rot=<0.0237 0.0095 -0.0035 -0.9997>}
 deform_knee_in_base_l {parent=calf_l, loc=<0.8031 -5.0231 0.2456>, rot=<0 0 0 -1>}
 deform_knee_in_l {parent=deform_knee_in_base_l, loc=<0 0 0>, rot=<0 0 0 1>}
 deform_knee_out_base_l {parent=calf_l, loc=<0 4 0>, rot=<0 0 0 -1>}
 deform_knee_out_l {parent=deform_knee_out_base_l, loc=<0 0 0>, rot=<0 0 0 1>}
 thigh_twist_01_l {parent=thigh_l, loc=<-27.952 0 0>, rot=<0.0013 0 0 -1>}
 deform_groin_base_l {parent=thigh_l, loc=<0.4801 5.6776 -0.2486>, rot=<0 0 0 -1>}
 deform_groin_l {parent=deform_groin_base_l, loc=<0 0 0>, rot=<0 0 0 1>}
 thigh_r {parent=pelvis, loc=<-1.0855 -0.2109 8.7053>, rot=<0.0355 -0.0009 0.9993 0.0155>}
 calf_r {parent=thigh_r, loc=<42.5284 0.0019 0.7297>, rot=<0.0006 -0.0161 -0.0391 -0.9991>}
 foot_r {parent=calf_r, loc=<46.7401 2.5014 0.592>, rot=<0.0427 -0.0238 0.0353 -0.9982>}
 ball_r {parent=foot_r, loc=<3.5658 -10.8066 0.1544>, rot=<0.0001 -0.0001 -0.7063 -0.7079>}
 calf_twist_02_r {parent=calf_r, loc=<19.1546 1.7943 0.2638>, rot=<0.0237 0.0095 -0.0035 -0.9997>}
 calf_twist_01_r {parent=calf_r, loc=<35.8483 1.9197 0.5794>, rot=<0.0237 0.0095 -0.0035 -0.9997>}
 deform_knee_in_base_r {parent=calf_r, loc=<-0.8031 5.0231 -0.2457>, rot=<0 0 0 -1>}
 deform_knee_in_r {parent=deform_knee_in_base_r, loc=<0 0 0>, rot=<0 0 0 1>}
 deform_knee_out_base_r {parent=calf_r, loc=<0 -4 0>, rot=<0 0 0 -1>}
 deform_knee_out_r {parent=deform_knee_out_base_r, loc=<0 0 0>, rot=<0 0 0 1>}
 thigh_twist_01_r {parent=thigh_r, loc=<27.952 0 0>, rot=<0.0013 0 0 -1>}
 deform_groin_base_r {parent=thigh_r, loc=<-0.4801 -5.6776 0.2486>, rot=<0 0 0 -1>}
 deform_groin_r {parent=deform_groin_base_r, loc=<0 0 0>, rot=<0 0 0 1>}
 deform_glute_base_l {parent=pelvis, loc=<-7.3466 -9.0213 -9.1245>, rot=<0.0172 0.031 -0.4477 -0.8935>}
 deform_glute_l {parent=deform_glute_base_l, loc=<0 0 0>, rot=<0 0 0 1>}
 deform_glute_base_r {parent=pelvis, loc=<-7.3464 -9.0213 9.1245>, rot=<0.031 -0.0172 0.8935 -0.4477>}
 deform_glute_r {parent=deform_glute_base_r, loc=<0 0 0>, rot=<0 0 0 1>}
 dyn_hair_0_l_ctrl_offset {type=Null, loc=<7.0715 2.9994 156.302>, rot=<-0.7046 0.0587 0.7071 -0.0127>}
 dyn_hair_0_l_ctrl {controlType=EULER_TRANSFORM, parent=dyn_hair_0_l_ctrl_offset, type=Control, loc=<0 0 0>, rot=<0 0 0 1>}
 dyn_hair_1_l_ctrl_offset {type=Null, loc=<7.0185 2.2089 148.4833>, rot=<0.7435 -0.021 -0.6679 -0.0253>}
 dyn_hair_1_l_ctrl {controlType=EULER_TRANSFORM, parent=dyn_hair_1_l_ctrl_offset, type=Control, loc=<0 0 0>, rot=<0 0 0 1>}
 dyn_hair_2_l_ctrl_offset {type=Null, loc=<7.8545 2.2287 140.7117>, rot=<-0.8188 -0.0209 0.5676 0.0834>}
 dyn_hair_2_l_ctrl {controlType=EULER_TRANSFORM, parent=dyn_hair_2_l_ctrl_offset, type=Control, loc=<0 0 0>, rot=<0 0 0 1>}
 dyn_hair_3_l_ctrl_offset {type=Null, loc=<10.692 3.2595 133.3066>, rot=<0.763 0.0181 -0.6418 -0.0753>}
 dyn_hair_3_l_ctrl {controlType=EULER_TRANSFORM, parent=dyn_hair_3_l_ctrl_offset, type=Control, loc=<0 0 0>, rot=<0 0 0 1>}
 dyn_hair_4_l_ctrl_offset {type=Null, loc=<12.0639 4.2309 125.673>, rot=<-0.6725 0.0584 0.7372 0.0303>}
 dyn_hair_4_l_ctrl {controlType=EULER_TRANSFORM, parent=dyn_hair_4_l_ctrl_offset, type=Control, loc=<0 0 0>, rot=<0 0 0 1>}
 dyn_hair_5_l_ctrl_offset {type=Null, loc=<11.3317 3.9662 117.8954>, rot=<0.6725 -0.0584 -0.7372 -0.0303>}
 dyn_hair_5_l_ctrl {controlType=EULER_TRANSFORM, parent=dyn_hair_5_l_ctrl_offset, type=Control, loc=<0 0 0>, rot=<0 0 0 1>}
 dyn_hair_0_r_ctrl_offset {type=Null, loc=<-7.0715 2.9994 156.302>, rot=<0.0587 -0.7046 0.0127 -0.7071>}
 dyn_hair_0_r_ctrl {controlType=EULER_TRANSFORM, parent=dyn_hair_0_r_ctrl_offset, type=Control, loc=<0 0 0>, rot=<0 0 0 1>}
 dyn_hair_1_r_ctrl_offset {type=Null, loc=<-7.0185 2.2089 148.483>, rot=<-0.021 0.7435 0.0253 0.6679>}
 dyn_hair_1_r_ctrl {controlType=EULER_TRANSFORM, parent=dyn_hair_1_r_ctrl_offset, type=Control, loc=<0 0 0>, rot=<0 0 0 1>}
 dyn_hair_2_r_ctrl_offset {type=Null, loc=<-7.8545 2.2287 140.712>, rot=<-0.0209 -0.8188 -0.0834 -0.5676>}
 dyn_hair_2_r_ctrl {controlType=EULER_TRANSFORM, parent=dyn_hair_2_r_ctrl_offset, type=Control, loc=<0 0 0>, rot=<0 0 0 1>}
 dyn_hair_3_r_ctrl_offset {type=Null, loc=<-10.692 3.2595 133.307>, rot=<0.0181 0.763 0.0753 0.6418>}
 dyn_hair_3_r_ctrl {controlType=EULER_TRANSFORM, parent=dyn_hair_3_r_ctrl_offset, type=Control, loc=<0 0 0>, rot=<0 0 0 1>}
 dyn_hair_4_r_ctrl_offset {type=Null, loc=<-12.0639 4.2309 125.673>, rot=<0.0584 -0.6725 -0.0303 -0.7372>}
 dyn_hair_4_r_ctrl {controlType=EULER_TRANSFORM, parent=dyn_hair_4_r_ctrl_offset, type=Control, loc=<0 0 0>, rot=<0 0 0 1>}
 dyn_hair_5_r_ctrl_offset {type=Null, loc=<-11.3317 3.9662 117.895>, rot=<-0.0584 0.6725 0.0303 0.7372>}
 dyn_hair_5_r_ctrl {controlType=EULER_TRANSFORM, parent=dyn_hair_5_r_ctrl_offset, type=Control, loc=<0 0 0>, rot=<0 0 0 1>}
 spine_01_ctrl {controlType=EULER_TRANSFORM, type=Control, loc=<0 0 0>, rot=<0 0 0 1>}
 spine_02_ctrl {controlType=EULER_TRANSFORM, parent=spine_01_ctrl, type=Control, loc=<0 0 0>, rot=<0 0 0 1>}
 spine_03_ctrl {controlType=EULER_TRANSFORM, parent=spine_02_ctrl, type=Control, loc=<0 0 0>, rot=<0 0 0 1>}
 spine_04_ctrl {controlType=EULER_TRANSFORM, parent=spine_03_ctrl, type=Control, loc=<0 0 0>, rot=<0 0 0 1>}
 spine_05_ctrl {controlType=EULER_TRANSFORM, parent=spine_04_ctrl, type=Control, loc=<0 0 0>, rot=<0 0 0 1>}
 neck_01_ctrl {controlType=EULER_TRANSFORM, parent=spine_05_ctrl, type=Control, loc=<0 0 0>, rot=<0 0 0 1>}
 neck_02_ctrl {controlType=EULER_TRANSFORM, parent=neck_01_ctrl, type=Control, loc=<0 0 0>, rot=<0 0 0 1>}
 head_ctrl {controlType=EULER_TRANSFORM, parent=neck_02_ctrl, type=Control, loc=<0 0 0>, rot=<0 0 0 1>}
 clavicle_l_ctrl {controlType=EULER_TRANSFORM, parent=spine_05_ctrl, type=Control, loc=<0 0 0>, rot=<0 0 0 1>}
 clavicle_r_ctrl {controlType=EULER_TRANSFORM, parent=spine_05_ctrl, type=Control, loc=<0 0 0>, rot=<0 0 0 1>}
]
